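# Control Rig for the M_MED_BananaAdventure character body that drives a duplicate dfrm_ deformation chain on the left arm.
# It copies arm rotations, solves PBIK to the hand, then sets forearm twist from swing-twist and constrains twist bones.
# A spherical pose reader multiplied by elbow-to-pelvis proximity weights a corrective upper-arm roll about the shoulder-to-hand axis.
# Right-arm copy, twist and roll nodes are unwired and target bones absent from the hierarchy.
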ControlRigBlueprint /BRCosmetics/Characters/Player/Male/Medium/Bodies/M_MED_BananaAdventure/Meshes/M_MED_BananaAdventure_CtrlRig
graphs [
 RigVMFunctionLibrary {links [], nodes []}
 RigVMModel {nodes [
   n1 = RigVMDispatchNode(
    ExecuteContext=()
    Array=&RigUnit_ItemArray(Items=((Type=Bone,Name="clavicle_r"),(Type=Bone,Name="upperarm_r"),(Type=Bone,Name="lowerarm_r"))).Items
   ) {template="DISPATCH_RigVMDispatch_ArrayIterator(in Array,out Element,out Index,out Count,out Ratio)"}
   > &RigUnit_SetRotation(
    Space=GlobalSpace
    Weight=1
    bInitial=False
    bPropagateToChildren=true
    Item=RigVMDispatchNode(
     Array=&RigUnit_ItemArray(Items=((Type=Bone,Name="dfrm_clavicle_r"),(Type=Bone,Name="dfrm_upperarm_r"),(Type=Bone,Name="dfrm_lowerarm_r"))).Items
     Index=@n1.Index
    ) {template="DISPATCH_RigVMDispatch_ArrayGetAtIndex(in Array,in Index,out Element)"}.Element
    Value=&RigUnit_GetTransform(Item=@n1.Element).Transform.Rotation
   )
   /Script/PBIK.RigUnit_PBIK(
    BoneSettings=((Bone="dfrm_lowerarm_r",X=Locked,Y=Locked,Z=Free,PreferredAngles=(X=0,Y=0,Z=0)))
    Debug=(DrawScale=1)
    Effectors=((Bone="dfrm_hand_r",Transform=(Rotation=(X=0,Y=0,Z=0,W=1),Translation=(X=0,Y=0,Z=0),Scale3D=(X=1,Y=1,Z=1)),PositionAlpha=1,RotationAlpha=1,StrengthAlpha=1,PinRotation=1,ChainDepth=0))
    ExcludedBones=()
    Root=dfrm_clavicle_l
    Settings="(Iterations=20,SubIterations=0,MassMultiplier=1,bAllowStretch=False,RootBehavior=PinToInput,PrePullRootSettings=(RotationAlpha=0,RotationAlphaX=1,RotationAlphaY=1,RotationAlphaZ=1,PositionAlpha=1,PositionAlphaX=1,PositionAlphaY=1,PositionAlphaZ=1),GlobalPu ...(truncated)"
    Effectors.0.Transform=&RigUnit_GetTransform(Item=(Type=Bone,Name="hand_r")).Transform
    ExecutePin=@n1.Completed
   ) {is=struct RigVMUnitNode}
   n2 = RigVMRerouteNode(Value=())
   n3 = &RigVMFunction_MathQuaternionSwingTwist(
    TwistAxis=(X=1,Y=0,Z=0)
    Input=&RigVMFunction_MathTransformMakeRelative(
     Parent=&RigUnit_GetTransform(Item=(Type=Bone,Name="hand_r"), Space=LocalSpace, bInitial=true).Transform
     Global=&RigUnit_GetTransform(Item=(Type=Bone,Name="hand_r"), Space=LocalSpace).Transform
    ).Local.Rotation
   )
   &RigUnit_SetRotation(Item=(Type=Bone,Name="dfrm_lowerarm_twist_01_r"), Weight=0.75, bPropagateToChildren=false, Value=@n3.Twist, ExecutePin=@n2.Value)
   > &RigUnit_SetRotation(Item=(Type=Bone,Name="dfrm_lowerarm_twist_02_r"), Weight=0.5, bPropagateToChildren=false, Value=@n3.Twist)
   RigUnit_BeginExecution = &RigUnit_BeginExecution()
   SequenceAggregate_1 = RigVMAggregateNode(ExecuteContext=@RigUnit_BeginExecution.ExecutePin)
   n4 = &RigUnit_SphericalPoseReader(
    ActiveRegionScaleFactors=(PositiveWidth=1,NegativeWidth=1,PositiveHeight=1.2,NegativeHeight=1.2)
    ActiveRegionSize=0.05
    Debug=(bDrawDebug=true,bDraw2D=False,bDrawLocalAxes=False,DebugScale=25,DebugSegments=20,DebugThickness=0.25)
    DriverAxis=(X=1,Y=0,Z=0)
    DriverItem=(Type=Bone,Name="upperarm_l")
    FalloffRegionScaleFactors=(PositiveWidth=0.5,NegativeWidth=0.75,PositiveHeight=0.8,NegativeHeight=0.5)
    FalloffSize=0.3
    FlipHeightScaling=False
    FlipWidthScaling=False
    OptionalParentItem=(Type=Bone,Name="None")
    RotationOffset=(X=0,Y=-155,Z=27.5)
    ExecutePin=@SequenceAggregate_1.A
   )
   > RigVMVariableNode(Value=@n4.OutputParam) {variable=SPR_upperarm_l_down_A_alpha}
   > n5 = &RigUnit_SphericalPoseReader(
    ActiveRegionScaleFactors=(PositiveWidth=1,NegativeWidth=1,PositiveHeight=1,NegativeHeight=1)
    ActiveRegionSize=0.065
    Debug=(bDrawDebug=false,bDraw2D=False,bDrawLocalAxes=False,DebugScale=25,DebugSegments=20,DebugThickness=0.25)
    DriverAxis=(X=-1,Y=0,Z=0)
    DriverItem=(Type=Bone,Name="upperarm_r")
    FalloffRegionScaleFactors=(PositiveWidth=0.75,NegativeWidth=1,PositiveHeight=0.75,NegativeHeight=1)
    FalloffSize=0.2
    FlipHeightScaling=False
    FlipWidthScaling=False
    OptionalParentItem=(Type=Bone,Name="None")
    RotationOffset=(X=0,Y=50,Z=-7.5)
   )
   > RigVMVariableNode(Value=@n5.OutputParam) {variable=SPR_upperarm_r_down_A_alpha}
   > n6 = &RigUnit_SphericalPoseReader(
    ActiveRegionScaleFactors=(PositiveWidth=1,NegativeWidth=1,PositiveHeight=1,NegativeHeight=1)
    ActiveRegionSize=0.05
    Debug=(bDrawDebug=false,bDraw2D=False,bDrawLocalAxes=true,DebugScale=25,DebugSegments=20,DebugThickness=0.25)
    DriverAxis=(X=-1,Y=0,Z=0)
    DriverItem=(Type=Bone,Name="upperarm_r")
    FalloffRegionScaleFactors=(PositiveWidth=0.5,NegativeWidth=1,PositiveHeight=0.75,NegativeHeight=1)
    FalloffSize=0.2
    FlipHeightScaling=False
    FlipWidthScaling=False
    OptionalParentItem=(Type=Bone,Name="None")
    RotationOffset=(X=-23.5,Y=17.5,Z=0)
   )
   > RigVMVariableNode(Value=@n6.OutputParam) {variable=SPR_upperarm_r_down_B_alpha}
   SequenceAggregate = &RigVMFunction_Sequence(ExecuteContext=@SequenceAggregate_1.C)
   RigUnit_ItemArray = &RigUnit_ItemArray(Items=((Type=Bone,Name="clavicle_l"),(Type=Bone,Name="upperarm_l"),(Type=Bone,Name="lowerarm_l")))
   DISPATCH_RigVMDispatch_ArrayIterator = RigVMDispatchNode(Array=@RigUnit_ItemArray.Items, ExecuteContext=@SequenceAggregate.A) {template="DISPATCH_RigVMDispatch_ArrayIterator(in Array,out Element,out Index,out Count,out Ratio)"}
   DISPATCH_RigVMDispatch_ArrayGetAtIndex = RigVMDispatchNode(
    Array=&RigUnit_ItemArray(Items=((Type=Bone,Name="dfrm_clavicle_l"),(Type=Bone,Name="dfrm_upperarm_l"),(Type=Bone,Name="dfrm_lowerarm_l"))).Items
    Index=@DISPATCH_RigVMDispatch_ArrayIterator.Index
   ) {template="DISPATCH_RigVMDispatch_ArrayGetAtIndex(in Array,in Index,out Element)"}
   > &RigUnit_SetRotation(
    Space=GlobalSpace
    Weight=1
    bInitial=False
    bPropagateToChildren=true
    Item=@DISPATCH_RigVMDispatch_ArrayGetAtIndex.Element
    Value=&RigUnit_GetTransform(Item=@DISPATCH_RigVMDispatch_ArrayIterator.Element).Transform.Rotation
   )
   PBIK = &RigUnit_PBIK(
    BoneSettings=((Bone="dfrm_lowerarm_l",X=Locked,Y=Locked,Z=Free,PreferredAngles=(X=0,Y=0,Z=0)))
    Debug=(DrawScale=1)
    Effectors=((Bone="dfrm_hand_l",Transform=(Rotation=(X=0,Y=0,Z=0,W=1),Translation=(X=0,Y=0,Z=0),Scale3D=(X=1,Y=1,Z=1)),PositionAlpha=1,RotationAlpha=1,StrengthAlpha=1,PinRotation=1,ChainDepth=0))
    ExcludedBones=()
    Root=dfrm_clavicle_l
    Settings="(Iterations=20,SubIterations=0,MassMultiplier=1,bAllowStretch=False,RootBehavior=PinToInput,PrePullRootSettings=(RotationAlpha=0,RotationAlphaX=1,RotationAlphaY=1,RotationAlphaZ=1,PositionAlpha=1,PositionAlphaX=1,PositionAlphaY=1,PositionAlphaZ=1),GlobalPu ...(truncated)"
    Effectors.0.Transform=&RigUnit_GetTransform(Item=(Type=Bone,Name="hand_l")).Transform
    ExecutePin=@DISPATCH_RigVMDispatch_ArrayIterator.Completed
   )
   SequenceAggregate_2 = &RigVMFunction_Sequence(ExecuteContext=@SequenceAggregate_1.D)
   Subtract = /Script/RigVM.RigVMFunction_MathVectorSub(
    B=&RigUnit_GetTransform(Item=(Type=Bone,Name="dfrm_upperarm_l")).Transform.Translation
    A=&RigUnit_GetTransform(Item=(Type=Bone,Name="dfrm_hand_l")).Transform.Translation
   ) {is=struct RigVMUnitNode, template="Subtract::Execute(in A,in B,out Result)"}
   Unit = /Script/RigVM.RigVMFunction_MathVectorUnit(Value=@Subtract.Result) {is=struct RigVMUnitNode, template="Unit::Execute(in Value,out Result)"}
   RigVMFunction_MathQuaternionFromAxisAndAngle = /Script/RigVM.RigVMFunction_MathQuaternionFromAxisAndAngle(Angle=-0.8, Axis=@Unit.Result) {is=struct RigVMUnitNode, template="FromAxisAndAngle::Execute(in Axis,in Angle,out Result)"}
   Set Transform = &RigUnit_SetRotation(
    Item=(Type=Bone,Name="dfrm_upperarm_l")
    Space=GlobalSpace
    bInitial=false
    Value=/Script/RigVM.RigVMFunction_MathQuaternionMul(
     B=&RigUnit_GetTransform(Item=(Type=Bone,Name="dfrm_upperarm_l")).Transform.Rotation
     A=@RigVMFunction_MathQuaternionFromAxisAndAngle.Result
    ) {is=struct RigVMUnitNode, template="Multiply::Execute(in A,in B,out Result)"}.Result
    ExecutePin=@SequenceAggregate_2.A
    Weight=RigVMVariableNode() {variable=mult_spr_dst_lowerarm_r_pelvis_alpha}.Value
   )
   Remap = &RigVMFunction_MathDoubleRemap(
    SourceMaximum=40
    SourceMinimum=25
    TargetMaximum=1
    TargetMinimum=0
    Value=&RigVMFunction_MathVectorDistance(
     A=&RigUnit_GetTransform(Item=(Type=Bone,Name="pelvis"), bInitial=False).Transform.Translation
     B=&RigUnit_GetTransform(Item=(Type=Bone,Name="lowerarm_l"), bInitial=False).Transform.Translation
    ).Result
   )
   Multiply = &RigVMFunction_MathDoubleMul(B=&RigVMFunction_MathDoubleSub(A=1, B=@Remap.Result).Result, A=RigVMVariableNode() {variable=SPR_upperarm_l_down_A_alpha}.Value)
   RigVMVariableNode(Value=@Multiply.Result, ExecuteContext=@SequenceAggregate_1.B) {variable=mult_spr_dst_lowerarm_r_pelvis_alpha}
   RigVMFunction_MathVectorDistance = &RigVMFunction_MathVectorDistance(
    A=&RigUnit_GetTransform(Item=(Type=Bone,Name="hand_r"), bInitial=False).Transform.Translation
    B=&RigUnit_GetTransform(Item=(Type=Bone,Name="thigh_twist_01_r"), bInitial=False).Transform.Translation
   )
   > RigVMVariableNode(Value=&RigVMFunction_MathDoubleSub(
     A=1
     B=&RigVMFunction_MathDoubleRemap(SourceMaximum=59, SourceMinimum=19.5, TargetMaximum=1, TargetMinimum=0, Value=@RigVMFunction_MathVectorDistance.Result).Result
    ).Result) {variable=mult_spr_dst_hand_r_thigh_twist_r_alpha}
   n7 = &RigVMFunction_MathQuaternionSwingTwist(
    TwistAxis=(X=1,Y=0,Z=0)
    Input=&RigVMFunction_MathTransformMakeRelative(
     Parent=&RigUnit_GetTransform(Item=(Type=Bone,Name="hand_l"), Space=LocalSpace, bInitial=true).Transform
     Global=&RigUnit_GetTransform(Item=(Type=Bone,Name="hand_l"), Space=LocalSpace).Transform
    ).Local.Rotation
   )
   &RigUnit_SetRotation(
    Item=(Type=Bone,Name="dfrm_lowerarm_twist_01_l")
    Weight=0.75
    bPropagateToChildren=false
    Value=@n7.Twist
    ExecutePin=@SequenceAggregate_1.E
   )
   > &RigUnit_SetRotation(Item=(Type=Bone,Name="dfrm_lowerarm_twist_02_l"), Weight=0.5, bPropagateToChildren=false, Value=@n7.Twist)
   > &RigUnit_ParentConstraint(
    AdvancedSettings=(InterpolationType=Average,RotationOrderForFilter=XZY)
    Child=(Type=Bone,Name="dfrm_lowerarm_twist_01_l")
    Filter=(TranslationFilter=(bX=True,bY=True,bZ=True),RotationFilter=(bX=false,bY=false,bZ=false),ScaleFilter=(bX=True,bY=True,bZ=True))
    Parents=((Item=(Type=Bone,Name="lowerarm_twist_01_l"),Weight=1))
    Weight=1
   )
   > &RigUnit_ParentConstraint(
    AdvancedSettings=(InterpolationType=Average,RotationOrderForFilter=XZY)
    Child=(Type=Bone,Name="dfrm_lowerarm_twist_02_l")
    Filter=(TranslationFilter=(bX=True,bY=True,bZ=True),RotationFilter=(bX=false,bY=false,bZ=false),ScaleFilter=(bX=True,bY=True,bZ=True))
    Parents=((Item=(Type=Bone,Name="lowerarm_twist_02_l"),Weight=1))
    Weight=1
   )
   > &RigUnit_ParentConstraint(
    AdvancedSettings=(InterpolationType=Average,RotationOrderForFilter=XZY)
    Child=(Type=Bone,Name="dfrm_upperarm_twist_01_l")
    Filter=(TranslationFilter=(bX=True,bY=True,bZ=True),RotationFilter=(bX=true,bY=true,bZ=true),ScaleFilter=(bX=false,bY=false,bZ=false))
    Parents=((Item=(Type=Bone,Name="upperarm_twist_01_l"),Weight=1))
    Weight=1
   )
   > &RigUnit_ParentConstraint(
    AdvancedSettings=(InterpolationType=Average,RotationOrderForFilter=XZY)
    Child=(Type=Bone,Name="dfrm_upperarm_twist_02_l")
    Filter=(TranslationFilter=(bX=True,bY=True,bZ=True),RotationFilter=(bX=true,bY=true,bZ=true),ScaleFilter=(bX=false,bY=false,bZ=false))
    Parents=((Item=(Type=Bone,Name="upperarm_twist_02_l"),Weight=1))
    Weight=1
   )
   n8 = &RigVMFunction_MathVectorUnit(Value=&RigVMFunction_MathVectorSub(
     B=&RigUnit_GetTransform(Item=(Type=Bone,Name="dfrm_upperarm_r"), bInitial=False).Transform.Translation
     A=&RigUnit_GetTransform(Item=(Type=Bone,Name="dfrm_hand_r"), bInitial=False).Transform.Translation
    ) {template="Subtract::Execute(in A,in B,out Result)"}.Result) {template="Unit::Execute(in Value,out Result)"}
   Set Transform_1 = &RigUnit_SetRotation(
    ExecutePin=()
    Item=(Type=Bone,Name="dfrm_upperarm_r")
    Space=GlobalSpace
    bInitial=False
    Value=&RigVMFunction_MathQuaternionMul(
     B=&RigUnit_GetTransform(Item=(Type=Bone,Name="dfrm_upperarm_r"), bInitial=False).Transform.Rotation
     A=&RigVMFunction_MathQuaternionFromAxisAndAngle(Angle=0.35, Axis=@n8.Result) {template="FromAxisAndAngle::Execute(in Axis,in Angle,out Result)"}.Result
    ) {template="Multiply::Execute(in A,in B,out Result)"}.Result
    Weight=RigVMVariableNode() {variable=SPR_upperarm_r_down_B_alpha}.Value
   )
  ]}
 SequenceAggregate_1_1_ContainedGraph {nodes [
   n1 = RigVMFunctionEntryNode()
   SequenceAggregate_1 = &RigVMFunction_Sequence(ExecuteContext=@n1.ExecuteContext)
   SequenceAggregate_1_1 = &RigVMFunction_Sequence(ExecuteContext=@SequenceAggregate_1.B)
   SequenceAggregate_1_2 = &RigVMFunction_Sequence(ExecuteContext=@SequenceAggregate_1_1.B)
   SequenceAggregate_1_3 = &RigVMFunction_Sequence(ExecuteContext=@SequenceAggregate_1_2.B)
   SequenceAggregate_1_4 = &RigVMFunction_Sequence(ExecuteContext=@SequenceAggregate_1_3.B)
   RigVMFunctionReturnNode(
    A=@SequenceAggregate_1.A
    B=@SequenceAggregate_1_1.A
    C=@SequenceAggregate_1_2.A
    D=@SequenceAggregate_1_3.A
    E=@SequenceAggregate_1_4.A
    F=@SequenceAggregate_1_4.B
   )
  ]}
]
hierarchy [
 root {loc=<0 0 0>, rot=<0 0 0 1>}
 pelvis {parent=root, loc=<0 -0.7071 93.1191>, rot=<0.0127 0.707 -0.0127 -0.707>}
 spine_01 {parent=pelvis, loc=<6.7494 -0.6388 0>, rot=<0 0 0.034 -0.9994>}
 spine_02 {parent=spine_01, loc=<7.2518 -0.035 0>, rot=<0 0 0.0238 -0.9997>}
 spine_03 {parent=spine_02, loc=<6.0104 0 0>, rot=<0 0 0.0149 -0.9999>}
 spine_04 {parent=spine_03, loc=<8.1024 0 0>, rot=<0 0 -0.0132 -0.9999>}
 spine_05 {parent=spine_04, loc=<10.0146 0 0>, rot=<0 0 -0.0015 -1>}
 attach_cape {parent=spine_05, loc=<9.017 -11.714 0>, rot=<0.0616 -0.7044 -0.0616 -0.7044>}
 neck_01 {parent=spine_05, loc=<11.3138 0.1529 0>, rot=<0 0 -0.2003 -0.9797>}
 neck_02 {parent=neck_01, loc=<4.5627 0 0>, rot=<0 0 0.0504 -0.9987>}
 head {parent=neck_02, loc=<5.3914 0 0>, rot=<0 0 0.115 -0.9934>}
 attach_backpack {parent=spine_05, loc=<-0.123 -13.473 0>, rot=<-0.0093 -0.707 0.0093 -0.707>}
 clavicle_l {parent=spine_05, loc=<4.3575 -0.1466 -2.3223>, rot=<0.0465 -0.7289 -0.009 -0.6829>}
 upperarm_l {parent=clavicle_l, loc=<15.5111 0 0>, rot=<0.0226 -0.3234 -0.0088 -0.9459>}
 lowerarm_l {parent=upperarm_l, loc=<26.6514 0 0>, rot=<0.0391 -0.0099 -0.2453 -0.9686>}
 hand_l {parent=lowerarm_l, loc=<22.3491 0 0>, rot=<-0.7081 -0.0507 0.0975 -0.6975>}
 thumb_01_l {parent=hand_l, loc=<1.3118 -0.3191 -2.459>, rot=<0.6546 -0.4042 0.0664 -0.6354>}
 thumb_02_l {parent=thumb_01_l, loc=<5.9336 -0.0448 0.0401>, rot=<-0.0037 0.0213 0.298 -0.9543>}
 thumb_03_l {parent=thumb_02_l, loc=<2.2223 0.3682 -0.0907>, rot=<0.0118 -0.013 0.0066 -0.9998>}
 index_metacarpal_l {parent=hand_l, loc=<1.6535 0.0183 -2.96>, rot=<0.0158 -0.0124 -0.0173 -0.9996>}
 index_01_l {parent=index_metacarpal_l, loc=<8.3008 0.0357 -0.0524>, rot=<-0.0044 0.0485 0.0924 -0.9945>}
 index_02_l {parent=index_01_l, loc=<3.3587 -0.3004 0.0452>, rot=<-0.002 0.0027 0.3225 -0.9466>}
 index_03_l {parent=index_02_l, loc=<2.5578 -0.123 -0.0252>, rot=<-0.002 0.0067 0.2558 -0.9667>}
 middle_metacarpal_l {parent=hand_l, loc=<2.1856 -0.1725 -0.5984>, rot=<0.006 0.0168 -0.0329 -0.9993>}
 middle_01_l {parent=middle_metacarpal_l, loc=<7.3275 -0.2447 0.0129>, rot=<-0.0208 0.0042 0.0858 -0.9961>}
 middle_02_l {parent=middle_01_l, loc=<4.0123 0.1401 0.1175>, rot=<-0.0218 0.0258 0.3512 -0.9357>}
 middle_03_l {parent=middle_02_l, loc=<2.6077 0.0534 0.1373>, rot=<0.0082 0.0188 0.1708 -0.9851>}
 ring_metacarpal_l {parent=hand_l, loc=<2.2074 -0.3625 1.3638>, rot=<-0.0284 0.0722 -0.0458 -0.9959>}
 ring_01_l {parent=ring_metacarpal_l, loc=<7.0295 -0.6197 -0.1446>, rot=<-0.0489 -0.0252 0.1892 -0.9804>}
 ring_02_l {parent=ring_01_l, loc=<3.6174 0.4647 -0.1027>, rot=<0 0 0.2806 -0.9598>}
 ring_03_l {parent=ring_02_l, loc=<2.4152 0.114 0.0199>, rot=<0 0 0.1189 -0.9929>}
 pinky_metacarpal_l {parent=hand_l, loc=<1.7616 -0.9753 3.3814>, rot=<-0.0397 0.0814 -0.069 -0.9935>}
 pinky_01_l {parent=pinky_metacarpal_l, loc=<7.1709 -0.4831 -0.1094>, rot=<-0.0654 -0.0035 0.2242 -0.9723>}
 pinky_02_l {parent=pinky_01_l, loc=<2.8274 0.0893 -0.1043>, rot=<0.0037 0.0011 0.2854 -0.9584>}
 pinky_03_l {parent=pinky_02_l, loc=<2.2046 0.0307 -0.0078>, rot=<-0.0002 0.0098 0.1445 -0.9895>}
 deform_wrist_dn_base_l {parent=hand_l, loc=<0 -2.2342 0>, rot=<0 0 0 1>}
 deform_wrist_dn_l {parent=deform_wrist_dn_base_l, loc=<0 0 0>, rot=<0 0 0 1>}
 deform_wrist_up_base_l {parent=hand_l, loc=<0 2.2682 0>, rot=<0 0 0 1>}
 deform_wrist_up_l {parent=deform_wrist_up_base_l, loc=<0 0 0>, rot=<0 0 0 1>}
 weapon_l {parent=hand_l, loc=<-0.1974 -3.4111 0.3811>, rot=<0 0 0 1>}
 hand_attach_l {parent=hand_l, loc=<8.5 -3 0>, rot=<0.7071 0.7071 0 0>}
 lowerarm_twist_02_l {parent=lowerarm_l, loc=<9.6199 0 0>, rot=<0 0 0 -1>}
 lowerarm_twist_01_l {parent=lowerarm_l, loc=<17.8549 0 0>, rot=<0 0 0 -1>}
 deform_elbow_in_base_l {parent=lowerarm_l, loc=<0.2112 5.259 0.8274>, rot=<0 0 0 -1>}
 deform_elbow_in_l {parent=deform_elbow_in_base_l, loc=<0 0 0>, rot=<0 0 0 1>}
 deform_elbow_out_base_l {parent=lowerarm_l, loc=<0 -4 0>, rot=<0 0 0 -1>}
 deform_elbow_out_l {parent=deform_elbow_out_base_l, loc=<0 0 0>, rot=<0 0 0 1>}
 upperarm_twist_01_l {parent=upperarm_l, loc=<0.0041 0 0>, rot=<0 0 0 1>}
 deform_upperarm_bk_base_l {parent=upperarm_twist_01_l, loc=<1.0813 -7.2193 -1.9453>, rot=<0 0 0 1>}
 deform_upperarm_bk_l {parent=deform_upperarm_bk_base_l, loc=<0 0 0>, rot=<0 0 0 1>}
 deform_upperarm_fr_base_l {parent=upperarm_twist_01_l, loc=<-0.3691 7.0506 -2.765>, rot=<0 0 0 1>}
 deform_upperarm_fr_l {parent=deform_upperarm_fr_base_l, loc=<0 0 0>, rot=<0 0 0 1>}
 deform_upperarm_up_base_l {parent=upperarm_twist_01_l, loc=<-3.2216 -1.1391 3.3311>, rot=<0 0 0 1>}
 deform_upperarm_up_l {parent=deform_upperarm_up_base_l, loc=<0 0 0>, rot=<0 0 0 1>}
 upperarm_twist_02_l {parent=upperarm_l, loc=<13.7941 0 0>, rot=<0 0 0 1>}
 clavicle_r {parent=spine_05, loc=<4.3575 -0.1466 2.3223>, rot=<0.7289 0.0465 -0.6829 0.009>}
 upperarm_r {parent=clavicle_r, loc=<-15.5111 0 0>, rot=<0.0226 -0.3234 -0.0088 -0.9459>}
 lowerarm_r {parent=upperarm_r, loc=<-26.6514 0 0>, rot=<0.0391 -0.0099 -0.2453 -0.9686>}
 hand_r {parent=lowerarm_r, loc=<-22.3491 0 0>, rot=<-0.7081 -0.0507 0.0975 -0.6975>}
 thumb_01_r {parent=hand_r, loc=<-1.3118 0.3191 2.459>, rot=<0.6546 -0.4042 0.0664 -0.6354>}
 thumb_02_r {parent=thumb_01_r, loc=<-5.9336 0.0448 -0.0401>, rot=<-0.0037 0.0213 0.298 -0.9543>}
 thumb_03_r {parent=thumb_02_r, loc=<-2.2223 -0.3682 0.0907>, rot=<0.0118 -0.013 0.0066 -0.9998>}
 index_metacarpal_r {parent=hand_r, loc=<-1.6535 -0.0183 2.96>, rot=<0.0158 -0.0124 -0.0173 -0.9996>}
 index_01_r {parent=index_metacarpal_r, loc=<-8.3008 -0.0357 0.0524>, rot=<-0.0044 0.0485 0.0924 -0.9945>}
 index_02_r {parent=index_01_r, loc=<-3.3587 0.3004 -0.0452>, rot=<-0.002 0.0027 0.3225 -0.9466>}
 index_03_r {parent=index_02_r, loc=<-2.5578 0.123 0.0252>, rot=<-0.002 0.0067 0.2558 -0.9667>}
 middle_metacarpal_r {parent=hand_r, loc=<-2.1856 0.1725 0.5984>, rot=<0.006 0.0168 -0.0329 -0.9993>}
 middle_01_r {parent=middle_metacarpal_r, loc=<-7.3275 0.2447 -0.0129>, rot=<-0.0208 0.0042 0.0858 -0.9961>}
 middle_02_r {parent=middle_01_r, loc=<-4.0123 -0.1401 -0.1175>, rot=<-0.0218 0.0258 0.3512 -0.9357>}
 middle_03_r {parent=middle_02_r, loc=<-2.6077 -0.0534 -0.1373>, rot=<0.0082 0.0188 0.1708 -0.9851>}
 ring_metacarpal_r {parent=hand_r, loc=<-2.2074 0.3625 -1.3638>, rot=<-0.0284 0.0722 -0.0458 -0.9959>}
 ring_01_r {parent=ring_metacarpal_r, loc=<-7.0295 0.6197 0.1446>, rot=<-0.0489 -0.0252 0.1892 -0.9804>}
 ring_02_r {parent=ring_01_r, loc=<-3.6174 -0.4647 0.1027>, rot=<0 0 0.2806 -0.9598>}
 ring_03_r {parent=ring_02_r, loc=<-2.4152 -0.114 -0.0199>, rot=<0 0 0.1189 -0.9929>}
 pinky_metacarpal_r {parent=hand_r, loc=<-1.7616 0.9753 -3.3814>, rot=<-0.0397 0.0814 -0.069 -0.9935>}
 pinky_01_r {parent=pinky_metacarpal_r, loc=<-7.1709 0.4831 0.1094>, rot=<-0.0654 -0.0035 0.2242 -0.9723>}
 pinky_02_r {parent=pinky_01_r, loc=<-2.8274 -0.0893 0.1043>, rot=<0.0037 0.0011 0.2854 -0.9584>}
 pinky_03_r {parent=pinky_02_r, loc=<-2.2046 -0.0307 0.0078>, rot=<-0.0002 0.0098 0.1445 -0.9895>}
 deform_wrist_dn_base_r {parent=hand_r, loc=<-0.0004 2.2345 -0.0001>, rot=<0 0 0 1>}
 deform_wrist_dn_r {parent=deform_wrist_dn_base_r, loc=<0 0 0>, rot=<0 0 0 1>}
 deform_wrist_up_base_r {parent=hand_r, loc=<0.0003 -2.2684 0.0001>, rot=<0 0 0 1>}
 deform_wrist_up_r {parent=deform_wrist_up_base_r, loc=<0 0 0>, rot=<0 0 0 1>}
 weapon_r {parent=hand_r, loc=<0.1974 3.4111 -0.3811>, rot=<0 0 0 1>}
 hand_attach_r {parent=hand_r, loc=<-8.5001 2.9999 0>, rot=<0 0 -0.7071 -0.7071>}
 lowerarm_twist_02_r {parent=lowerarm_r, loc=<-9.6199 0 0>, rot=<0 0 0 -1>}
 lowerarm_twist_01_r {parent=lowerarm_r, loc=<-17.8549 0 0>, rot=<0 0 0 -1>}
 deform_elbow_in_base_r {parent=lowerarm_r, loc=<-0.2111 -5.2591 -0.8275>, rot=<0 0 0 -1>}
 deform_elbow_in_r {parent=deform_elbow_in_base_r, loc=<-0.0001 0 0.0002>, rot=<0 0 0 1>}
 deform_elbow_out_base_r {parent=lowerarm_r, loc=<0 4 -0.0001>, rot=<0 0 0 -1>}
 deform_elbow_out_r {parent=deform_elbow_out_base_r, loc=<0 0 0>, rot=<0 0 0 1>}
 upperarm_twist_01_r {parent=upperarm_r, loc=<-0.0041 0 0>, rot=<0 0 0 1>}
 deform_upperarm_bk_base_r {parent=upperarm_twist_01_r, loc=<-1.0814 7.2192 1.9456>, rot=<0 0 0 1>}
 deform_upperarm_bk_r {parent=deform_upperarm_bk_base_r, loc=<0 0 0>, rot=<0 0 0 1>}
 deform_upperarm_fr_base_r {parent=upperarm_twist_01_r, loc=<0.3689 -7.0506 2.7652>, rot=<0 0 0 1>}
 deform_upperarm_fr_r {parent=deform_upperarm_fr_base_r, loc=<0 0 0>, rot=<0 0 0 1>}
 deform_upperarm_up_base_r {parent=upperarm_twist_01_r, loc=<3.2214 1.1391 -3.3308>, rot=<0 0 0 1>}
 deform_upperarm_up_r {parent=deform_upperarm_up_base_r, loc=<0 0 0>, rot=<0 0 0 -1>}
 upperarm_twist_02_r {parent=upperarm_r, loc=<-13.7941 0 0>, rot=<0 0 0 1>}
 dyn_jacket_g_constraint {parent=spine_05, loc=<-1.4773 19.5733 14.2325>, rot=<-0.1922 -0.6547 -0.1717 -0.7106>}
 dyn_jacket_g_MSTR {parent=dyn_jacket_g_constraint, loc=<0 0 -2.0095>, rot=<0.479 -0.4705 -0.5262 -0.5219>}
 dyn_jacket_g_01 {parent=dyn_jacket_g_MSTR, loc=<0 0 0>, rot=<0 0 0 1>}
 dyn_jacket_g_02 {parent=dyn_jacket_g_01, loc=<8.4676 -0.1667 -0.8278>, rot=<0.1132 -0.0119 -0.0079 -0.9935>}
 dyn_jacket_g_03 {parent=dyn_jacket_g_02, loc=<6.3398 -0.0462 -0.1896>, rot=<0.1191 0.0569 -0.0068 -0.9912>}
 q {parent=dyn_jacket_g_MSTR, loc=<0 0 0>, rot=<0 0 0 1>}
 dyn_jacket_a_constraint {parent=spine_05, loc=<-1.0888 18.7802 -13.0854>, rot=<0.1598 -0.6949 0.2469 -0.6562>}
 dyn_jacket_a_MSTR {parent=dyn_jacket_a_constraint, loc=<-0.0145 -0.1451 -2.5161>, rot=<0.4449 -0.553 -0.4323 -0.5562>}
 dyn_jacket_a_01 {parent=dyn_jacket_a_MSTR, loc=<0 0 0>, rot=<0 0 0 1>}
 dyn_jacket_a_02 {parent=dyn_jacket_a_01, loc=<8.7203 -0.0924 0.2725>, rot=<-0.0663 0.0533 0.0449 -0.9954>}
 dyn_Zipper_MSTR {parent=dyn_jacket_a_02, loc=<7.9809 0.4584 -1.0896>, rot=<-0.1573 -0.1498 -0.0695 -0.9736>}
 dyn_Zipper_01 {parent=dyn_Zipper_MSTR, loc=<0 0 0>, rot=<0 0 0 1>}
 dyn_scarf_MSTR {parent=spine_05, loc=<1.5529 19.2393 -0.0641>, rot=<-0.0276 -0.7075 -0.7026 -0.0716>}
 dyn_scarf_01 {parent=dyn_scarf_MSTR, loc=<0 0 0>, rot=<0 0 0 1>}
 dyn_scarf_02 {parent=dyn_scarf_01, loc=<9.2679 -0.1758 -0.9322>, rot=<0.0181 -0.0946 -0.0167 -0.9952>}
 dyn_scarf_03 {parent=dyn_scarf_02, loc=<6.3398 -0.0462 -0.1896>, rot=<0.1191 0.0569 -0.0068 -0.9912>}
 deform_neckline_l_01_root {parent=spine_05, loc=<6.5132 16.8248 -15.047>, rot=<0 0 0.0015 -1>}
 deform_neckline_l_01 {parent=deform_neckline_l_01_root, loc=<0 0 0>, rot=<0 0 0 1>}
 deform_neckline_l_02_root {parent=spine_05, loc=<10.5549 1.5702 -18.6728>, rot=<0 0 0.0015 -1>}
 deform_neckline_l_02 {parent=deform_neckline_l_02_root, loc=<0 0 0>, rot=<0 0 0 1>}
 deform_neckline_l_03_root {parent=spine_05, loc=<10.4865 -11.6231 -15.7294>, rot=<0 0 0.0015 -1>}
 deform_neckline_l_03 {parent=deform_neckline_l_03_root, loc=<0 0 0>, rot=<0 0 0 1>}
 deform_neckline_rear_root {parent=spine_05, loc=<11.146 -18.1745 0>, rot=<0 0 0.0015 -1>}
 deform_neckline_rear {parent=deform_neckline_rear_root, loc=<0 0 0>, rot=<0 0 0 1>}
 deform_neckline_r_01_root {parent=spine_05, loc=<5.0201 16.4963 15.9616>, rot=<0 0 1 0.0015>}
 deform_neckline_r_01 {parent=deform_neckline_r_01_root, loc=<0 0 0>, rot=<0 0 1 0>}
 deform_neckline_r_02_root {parent=spine_05, loc=<10.5379 1.5635 18.9186>, rot=<0 0 0.0015 -1>}
 deform_neckline_r_02 {parent=deform_neckline_r_02_root, loc=<0 0 0>, rot=<0 0 0 1>}
 deform_neckline_r_03_root {parent=spine_05, loc=<10.5701 -11.1093 15.8121>, rot=<0 0 0.0015 -1>}
 deform_neckline_r_03 {parent=deform_neckline_r_03_root, loc=<0 0 0>, rot=<0 0 0 1>}
 dfrm_clavicle_l {parent=spine_05, loc=<4.3575 -0.1466 -2.3223>, rot=<0.0465 -0.7289 -0.009 -0.6829>}
 dfrm_upperarm_l {parent=dfrm_clavicle_l, loc=<15.5111 0 0>, rot=<0.0226 -0.3234 -0.0088 -0.9459>}
 dfrm_lowerarm_l {parent=dfrm_upperarm_l, loc=<26.6514 0 0>, rot=<0.0391 -0.0099 -0.2453 -0.9686>}
 dfrm_hand_l {parent=dfrm_lowerarm_l, loc=<22.3491 0 0>, rot=<-0.7081 -0.0507 0.0975 -0.6975>}
 dfrm_lowerarm_twist_02_l {parent=dfrm_lowerarm_l, loc=<9.6199 0 0>, rot=<0 0 0 -1>}
 dfrm_lowerarm_twist_01_l {parent=dfrm_lowerarm_l, loc=<17.8549 0 0>, rot=<0 0 0 -1>}
 dfrm_deform_elbow_in_base_l {parent=dfrm_lowerarm_l, loc=<0.2112 5.259 0.8274>, rot=<0 0 0 -1>}
 dfrm_deform_elbow_in_l {parent=dfrm_deform_elbow_in_base_l, loc=<0 0 0>, rot=<0 0 0 1>}
 dfrm_deform_elbow_out_base_l {parent=dfrm_lowerarm_l, loc=<0 -4 0>, rot=<0 0 0 -1>}
 dfrm_deform_elbow_out_l {parent=dfrm_deform_elbow_out_base_l, loc=<0 0 0>, rot=<0 0 0 1>}
 dfrm_upperarm_twist_01_l {parent=dfrm_upperarm_l, loc=<0.0041 0 0>, rot=<0 0 0 1>}
 dfrm_deform_upperarm_bk_base_l {parent=dfrm_upperarm_twist_01_l, loc=<1.0813 -7.2193 -1.9453>, rot=<0 0 0 1>}
 dfrm_deform_upperarm_bk_l {parent=dfrm_deform_upperarm_bk_base_l, loc=<0 0 0>, rot=<0 0 0 1>}
 dfrm_deform_upperarm_fr_base_l {parent=dfrm_upperarm_twist_01_l, loc=<-0.3691 7.0506 -2.765>, rot=<0 0 0 1>}
 dfrm_deform_upperarm_fr_l {parent=dfrm_deform_upperarm_fr_base_l, loc=<0 0 0>, rot=<0 0 0 1>}
 dfrm_deform_upperarm_up_base_l {parent=dfrm_upperarm_twist_01_l, loc=<-3.2216 -1.1391 3.3311>, rot=<0 0 0 1>}
 dfrm_deform_upperarm_up_l {parent=dfrm_deform_upperarm_up_base_l, loc=<0 0 0>, rot=<0 0 0 1>}
 dfrm_upperarm_twist_02_l {parent=dfrm_upperarm_l, loc=<13.7941 0 0>, rot=<0 0 0 1>}
 deform_pec_aim_base_l {parent=spine_04, loc=<7.8732 12.6388 -2.4123>, rot=<0.0102 -0.4602 0.3979 -0.7936>}
 deform_pec_aim_l {parent=deform_pec_aim_base_l, loc=<0 0 0>, rot=<0 0 0 1>}
 deform_pec_base_l {parent=deform_pec_aim_l, loc=<10.766 0 0>, rot=<0 0 0 1>}
 deform_pec_l {parent=deform_pec_base_l, loc=<0.055 0 0>, rot=<0 0 0 1>}
 deform_pec_aim_base_r {parent=spine_04, loc=<7.8727 12.6388 2.4123>, rot=<-0.4602 -0.0102 0.7936 0.3979>}
 deform_pec_aim_r {parent=deform_pec_aim_base_r, loc=<0 0 0>, rot=<0 0 0 1>}
 deform_pec_base_r {parent=deform_pec_aim_r, loc=<-10.7662 -0.0006 -0.0007>, rot=<0 0 0 1>}
 deform_pec_r {parent=deform_pec_base_r, loc=<0 0 0>, rot=<0 0 0 1>}
 dyn_jacket_c_constraint {parent=spine_03, loc=<8.4529 -10.5661 -17.0617>, rot=<0.0263 -0.6972 -0.0287 -0.7158>}
 dyn_jacket_c_MSTR {parent=dyn_jacket_c_constraint, loc=<-0.2167 0.0914 -1.8263>, rot=<-0.2789 -0.6933 0.2796 -0.6028>}
 dyn_jacket_c_01 {parent=dyn_jacket_c_MSTR, loc=<0 0 0>, rot=<0 0 0 1>}
 dyn_jacket_c_02 {parent=dyn_jacket_c_01, loc=<10.3813 0 -0.0022>, rot=<0 0.0742 0 -0.9972>}
 dyn_jacket_b_constraint {parent=spine_03, loc=<9.0227 5.9296 -18.5824>, rot=<-0.0003 -0.7354 0.0015 -0.6776>}
 dyn_jacket_b_MSTR {parent=dyn_jacket_b_constraint, loc=<-0.0356 0.1356 -1.6873>, rot=<0.1073 -0.6979 -0.0804 -0.7035>}
 dyn_jacket_b_01 {parent=dyn_jacket_b_MSTR, loc=<0 0 0>, rot=<0 0 0 1>}
 dyn_jacket_b_02 {parent=dyn_jacket_b_01, loc=<10.3813 0 -0.0022>, rot=<0 0.0742 0 -0.9972>}
 dyn_jacket_d_constraint {parent=spine_03, loc=<10.4746 -21.686 -0.0086>, rot=<0.0387 -0.706 -0.0387 -0.706>}
 dyn_jacket_d_MSTR {parent=dyn_jacket_d_constraint, loc=<0.0289 0.0636 -1.4381>, rot=<0.5215 0.5003 -0.4895 0.4879>}
 dyn_jacket_d_01 {parent=dyn_jacket_d_MSTR, loc=<0 0 0>, rot=<0 0 0 1>}
 dyn_jacket_d_02 {parent=dyn_jacket_d_01, loc=<10.7944 0 -0.0171>, rot=<0 -0.018 0 -0.9998>}
 dyn_jacket_e_constraint {parent=spine_03, loc=<8.4534 -10.566 17.0617>, rot=<0.041 -0.7138 -0.0389 -0.6981>}
 dyn_jacket_e_MSTR {parent=dyn_jacket_e_constraint, loc=<0.151 0.2029 -1.4883>, rot=<0.6944 0.3182 -0.5905 0.2604>}
 dyn_jacket_e_01 {parent=dyn_jacket_e_MSTR, loc=<0 0 0>, rot=<0 0 0 1>}
 dyn_jacket_e_02 {parent=dyn_jacket_e_01, loc=<10.3813 0 -0.0022>, rot=<0 0.0742 0 -0.9972>}
 dyn_jacket_f_constraint {parent=spine_03, loc=<9.0227 5.9296 18.5824>, rot=<0.0407 -0.6595 -0.0318 -0.7499>}
 dyn_jacket_f_MSTR {parent=dyn_jacket_f_constraint, loc=<-0.144 0.025 -1.9072>, rot=<-0.6798 0.0068 0.7332 0.0152>}
 dyn_jacket_f_01 {parent=dyn_jacket_f_MSTR, loc=<0 0 0>, rot=<0 0 0 1>}
 dyn_jacket_f_02 {parent=dyn_jacket_f_01, loc=<10.2978 0 -0.2418>, rot=<0 -0.0277 0 -0.9996>}
 thigh_l {parent=pelvis, loc=<-2.5693 -0.0109 -7.7501>, rot=<0.0012 0.0572 0.0155 -0.9982>}
 calf_l {parent=thigh_l, loc=<-39.0484 0 0>, rot=<0.0006 -0.0161 -0.0391 -0.9991>}
 foot_l {parent=calf_l, loc=<-42.561 0.1077 0.4292>, rot=<0.0418 -0.0455 0.0343 -0.9975>}
 ball_l {parent=foot_l, loc=<-6.1599 13.8264 0.6806>, rot=<0.0001 -0.0001 -0.7063 -0.7079>}
 calf_twist_02_l {parent=calf_l, loc=<-18.8064 -0.0967 0.0397>, rot=<0.0237 0.0095 -0.0035 -0.9997>}
 calf_twist_01_l {parent=calf_l, loc=<-35.813 -0.0967 0.0397>, rot=<0.0237 0.0095 -0.0035 -0.9997>}
 deform_knee_in_base_l {parent=calf_l, loc=<1.298 -5.2957 -0.0496>, rot=<0 0 0 -1>}
 deform_knee_in_l {parent=deform_knee_in_base_l, loc=<0 0 0>, rot=<0 0 0 1>}
 deform_knee_out_base_l {parent=calf_l, loc=<0 4 0>, rot=<0 0 0 -1>}
 deform_knee_out_l {parent=deform_knee_out_base_l, loc=<0 0 0>, rot=<0 0 0 1>}
 thigh_twist_01_l {parent=thigh_l, loc=<-21.6303 0 0>, rot=<0.0013 0 0 -1>}
 dyn_shorts_main_l {parent=thigh_twist_01_l, loc=<-0.4614 -0.8676 1.1151>, rot=<-0.0027 -0.7464 0.0006 -0.6655>}
 dyn_shorts_out_l {parent=dyn_shorts_main_l, loc=<9.9469 -0.5145 1.2925>, rot=<0 0.1898 0 -0.9818>}
 dyn_shorts_out_02_l {parent=dyn_shorts_out_l, loc=<-2.685 0.2507 -13.1415>, rot=<0 0 0 1>}
 dyn_shorts_fr_l {parent=dyn_shorts_main_l, loc=<1.3227 13.2929 -1.7276>, rot=<-0.0067 0.0068 -0.7027 -0.7115>}
 dyn_shorts_fr_02_l {parent=dyn_shorts_fr_l, loc=<0 0 -12.0958>, rot=<0 0 0 1>}
 dyn_shorts_bk_l {parent=dyn_shorts_main_l, loc=<0.921 -10.0711 0.4559>, rot=<-0.0056 0.053 0.6996 -0.7125>}
 dyn_shorts_bk_02_l {parent=dyn_shorts_bk_l, loc=<-0.4147 0 -10.0565>, rot=<0 0 0 1>}
 dyn_shorts_in_l {parent=dyn_shorts_main_l, loc=<-8.6448 1.1393 -2.0369>, rot=<0.0069 0 1 0>}
 dyn_shorts_in_02_l {parent=dyn_shorts_in_l, loc=<-1.4897 0 -11.6935>, rot=<0 0 0 1>}
 deform_groin_base_l {parent=thigh_l, loc=<-2.1908 8.1142 -1.7442>, rot=<0 0 0 -1>}
 deform_groin_l {parent=deform_groin_base_l, loc=<0 0 0>, rot=<0 0 0 1>}
 thigh_r {parent=pelvis, loc=<-2.5693 -0.0109 7.7501>, rot=<0.0572 -0.0012 0.9982 0.0155>}
 calf_r {parent=thigh_r, loc=<39.0484 0 0>, rot=<0.0006 -0.0161 -0.0391 -0.9991>}
 foot_r {parent=calf_r, loc=<42.561 -0.1077 -0.4292>, rot=<0.0418 -0.0455 0.0343 -0.9975>}
 ball_r {parent=foot_r, loc=<6.1599 -13.8264 -0.6806>, rot=<0.0001 -0.0001 -0.7063 -0.7079>}
 calf_twist_02_r {parent=calf_r, loc=<18.8064 0.0967 -0.0397>, rot=<0.0237 0.0095 -0.0035 -0.9997>}
 calf_twist_01_r {parent=calf_r, loc=<35.813 0.0967 -0.0397>, rot=<0.0237 0.0095 -0.0035 -0.9997>}
 deform_knee_in_base_r {parent=calf_r, loc=<-1.298 5.2957 0.0496>, rot=<0 0 0 -1>}
 deform_knee_in_r {parent=deform_knee_in_base_r, loc=<0 0 0>, rot=<0 0 0 1>}
 deform_knee_out_base_r {parent=calf_r, loc=<0 -4 0>, rot=<0 0 0 -1>}
 deform_knee_out_r {parent=deform_knee_out_base_r, loc=<0 0 0>, rot=<0 0 0 1>}
 thigh_twist_01_r {parent=thigh_r, loc=<21.6303 0 0>, rot=<0.0013 0 0 -1>}
 dyn_shorts_main_r {parent=thigh_twist_01_r, loc=<0.1922 0.8688 -1.0841>, rot=<-0.0027 -0.7464 0.0006 -0.6655>}
 dyn_shorts_out_r {parent=dyn_shorts_main_r, loc=<-9.9468 0.5145 -1.2924>, rot=<0 0.1898 0 -0.9818>}
 dyn_shorts_out_02_r {parent=dyn_shorts_out_r, loc=<2.685 -0.2507 13.1415>, rot=<0 0 0 1>}
 dyn_shorts_fr_r {parent=dyn_shorts_main_r, loc=<-1.3238 -13.3818 1.7259>, rot=<-0.0067 0.0068 -0.7027 -0.7115>}
 dyn_shorts_fr_02_r {parent=dyn_shorts_fr_r, loc=<0 0 12.0958>, rot=<0 0 0 1>}
 dyn_shorts_bk_r {parent=dyn_shorts_main_r, loc=<-0.9179 9.8693 -0.4423>, rot=<-0.0056 0.053 0.6996 -0.7125>}
 dyn_shorts_bk_02_r {parent=dyn_shorts_bk_r, loc=<0.4147 0 10.0565>, rot=<0 0 0 1>}
 dyn_shorts_in_r {parent=dyn_shorts_main_r, loc=<8.6448 -1.1393 2.0369>, rot=<0.0069 0 1 0>}
 dyn_shorts_in_02_r {parent=dyn_shorts_in_r, loc=<1.4897 0 11.6936>, rot=<0 0 0 1>}
 deform_groin_base_r {parent=thigh_r, loc=<2.1908 -8.1142 1.7442>, rot=<0 0 0 1>}
 deform_groin_r {parent=deform_groin_base_r, loc=<0 0 0>, rot=<0 0 0 1>}
 deform_glute_base_l {parent=pelvis, loc=<-8.8064 -8.8222 -8.4287>, rot=<0.0275 0.0501 -0.4472 -0.8926>}
 deform_glute_l {parent=deform_glute_base_l, loc=<0 0 0>, rot=<0 0 0 1>}
 deform_glute_base_r {parent=pelvis, loc=<-8.8062 -8.8222 8.4287>, rot=<0.0501 -0.0275 0.8926 -0.4472>}
 deform_glute_r {parent=deform_glute_base_r, loc=<0 0 0>, rot=<0 0 0 1>}
 dyn_Map_MSTR {parent=pelvis, loc=<-8.0641 3.1683 18.1738>, rot=<0.11 -0.0361 0.9903 0.0763>}
 dyn_map {parent=dyn_Map_MSTR, loc=<0 0 0>, rot=<0 0 0 1>}
 dyn_mapFront {parent=dyn_map, loc=<0.5108 -2.1819 3.6567>, rot=<-0.6867 -0.042 0.0535 -0.7237>}
 dyn_mapBack {parent=dyn_map, loc=<0.5686 2.3771 3.4136>, rot=<0.7237 -0.0535 -0.042 -0.6867>}
 dyn_buckle_MSTR {parent=pelvis, loc=<-9.4751 20.7751 -0.007>, rot=<0.4328 0.5338 0.551 0.4735>}
 dyn_buckle_01 {parent=dyn_buckle_MSTR, loc=<0 0 0>, rot=<0 0 0 1>}
 attach {parent=root, loc=<0 0 0>, rot=<0 0 0 1>}
 ik_hand_root {parent=root, loc=<0 0 0>, rot=<0 0 0 1>}
 ik_hand_gun {parent=ik_hand_root, loc=<-51.9727 4.0883 102.6935>, rot=<0.6084 0.1793 0.4613 -0.6205>}
 ik_hand_r {parent=ik_hand_gun, loc=<0 0 0>, rot=<0 0 0 1>}
 ik_hand_l {parent=ik_hand_gun, loc=<53.0285 82.1781 35.2051>, rot=<-0.5102 -0.7906 -0.3387 0>}
 ik_foot_root {parent=root, loc=<0 0 0>, rot=<0 0 0 1>}
 ik_foot_l {parent=ik_foot_root, loc=<15.2608 -4.4176 9.4432>, rot=<0.0357 0.7045 0.0284 -0.7082>}
 ik_foot_r {parent=ik_foot_root, loc=<-15.2608 -4.4176 9.4432>, rot=<0.7082 0.0284 -0.7045 0.0357>}
 attach_fp {parent=root, loc=<0 0 0>, rot=<0 0 0 1>}
 camera_root {parent=root, loc=<0 0 0>, rot=<0 0 0 1>}
 camera {parent=camera_root, loc=<0 0 0>, rot=<0 0 0 1>}
 attach_emote_01 {parent=root, loc=<0 0 0>, rot=<0 0 0 1>}
 attach_emote_02 {parent=root, loc=<0 0 0>, rot=<0 0 0 1>}
 attach_emote_03 {parent=root, loc=<0 0 0>, rot=<0 0 0 1>}
 warp_point_01 {parent=root, loc=<0 0 0>, rot=<0 0 0 1>}
 warp_point_02 {parent=root, loc=<0 0 0>, rot=<0 0 0 1>}
 warp_point_03 {parent=root, loc=<0 0 0>, rot=<0 0 0 1>}
]
variables [
 SPR_upperarm_l_down_A_alpha {type=double, default=0}
 SPR_upperarm_r_down_A_alpha {type=double, default=0}
 SPR_upperarm_r_down_B_alpha {type=double, default=0}
 mult_spr_dst_hand_r_thigh_twist_r_alpha {type=double, default=0}
 mult_spr_dst_lowerarm_r_pelvis_alpha {type=double, default=0}
]
