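# Physics asset for the left shorts leg of the male medium BananaAdventure cosmetic body, providing bodies for dynamic cloth-like bone motion.
# Capsules cover both thighs and calves, plus a small main shorts hub and back, front, inner and outer flaps.
# The hub joins the left thigh twist bone with free angular motion.
# Each flap constrains to the hub with its own angle limits, widest at the back.
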PhysicsAsset /BRCosmetics/Characters/Player/Male/Medium/Bodies/M_MED_BananaAdventure/Meshes/M_MED_BananaAdventure_LeftShorts_Physics
bodies {
 calf_l [{shapeType=Capsule, length=16.3435, radius=7.1565, center=<-14.2533 -3.2563 -0.4527>, rotation=<rot 90 -3.4277 176.0191>}]
 calf_r [{shapeType=Capsule, length=15.7407, radius=7.7593, center=<14.1599 2.6497 0.7585>, rotation=<rot 90 -10.024 -10.3989>}]
 calf_twist_02_l [{shapeType=Capsule, length=17.9618, radius=5.5382, center=<-20.1486 -0.2148 0.664>, rotation=<rot 86.3626 -10.2246 -179.9995>}]
 calf_twist_02_r [{shapeType=Capsule, length=18.25, radius=5.25, center=<20.2357 0.8088 -0.0002>, rotation=<rot 90 0 3.5324>}]
 dyn_shorts_bk_l [{shapeType=Capsule, length=4.6484, radius=6.6402, shapeName=dyn_shorts_bk_l_capsule, center=<6 0 -6.4249>}]
 dyn_shorts_fr_l [{shapeType=Capsule, length=5.9262, radius=5.3624, shapeName=dyn_shorts_fr_l_capsule, center=<6 0 -6.4249>}]
 dyn_shorts_in_l [{shapeType=Capsule, length=5.0248, radius=6.2638, shapeName=dyn_shorts_in_l_capsule, center=<5 0 -8.1287>}]
 dyn_shorts_main_l [{shapeType=Capsule, length=0.505, radius=0.505, shapeName=dyn_shorts_main_l_capsule, rotation=<rot 0 -90 0>}]
 dyn_shorts_out_l [{shapeType=Capsule, length=5.0248, radius=6.2638, shapeName=dyn_shorts_out_l_capsule, center=<6 0 -6.3087>, rotation=<rot -13.3987 0 0>}]
 thigh_twist_01_l [{shapeType=Capsule, length=25.3009, radius=9.4491, center=<-2.3357 0.5566 0>, rotation=<rot 90 0 -5>}]
 thigh_twist_01_r [{shapeType=Capsule, length=26.25, radius=8.5, center=<2.3357 0.5566 0>, rotation=<rot 90 0 -5>}]
}
constraints [
 {
  bone1Name=calf_twist_02_l
  bone2Name=calf_l
  swing1LimitDegrees=45
  swing1Motion=Free
  swing2LimitDegrees=45
  swing2Motion=Free
  twistLimitDegrees=45
  twistMotion=Free
 }
 {
  bone1Name=calf_twist_02_r
  bone2Name=calf_r
  swing1LimitDegrees=45
  swing1Motion=Free
  swing2LimitDegrees=45
  swing2Motion=Free
  twistLimitDegrees=45
  twistMotion=Free
 }
 {
  bone1Name=dyn_shorts_main_l
  bone2Name=thigh_twist_01_l
  swing1LimitDegrees=45
  swing1Motion=Free
  swing2LimitDegrees=45
  swing2Motion=Free
  twistLimitDegrees=45
  twistMotion=Free
 }
 {bone1Name=dyn_shorts_bk_l, bone2Name=dyn_shorts_main_l, swing1LimitDegrees=60, swing2LimitDegrees=60, twistLimitDegrees=60}
 {bone1Name=dyn_shorts_fr_l, bone2Name=dyn_shorts_main_l, swing1LimitDegrees=10, swing2LimitDegrees=50, twistLimitDegrees=10}
 {bone1Name=dyn_shorts_in_l, bone2Name=dyn_shorts_main_l, swing1LimitDegrees=10, swing2LimitDegrees=50, twistLimitDegrees=10}
 {bone1Name=dyn_shorts_out_l, bone2Name=dyn_shorts_main_l, swing1LimitDegrees=10, swing2LimitDegrees=50, twistLimitDegrees=10}
]
