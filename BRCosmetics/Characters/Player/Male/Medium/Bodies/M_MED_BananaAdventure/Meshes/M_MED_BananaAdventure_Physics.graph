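# Physics asset for the M_MED_BananaAdventure male player body, covering its dynamic jacket bones and a zipper pull.
# Pelvis and spine capsules form the body core, with spine_05 anchoring seven jacket chains (a–g) through tiny constraint and MSTR capsules.
# Those anchor and spine joints rotate freely, while the large-capsule jacket segments below use tighter angular limits.
# Chain g has three segments, the others two, and the zipper hangs from the end of chain a.
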PhysicsAsset /BRCosmetics/Characters/Player/Male/Medium/Bodies/M_MED_BananaAdventure/Meshes/M_MED_BananaAdventure_Physics
bodies {
 dyn_Zipper_01 [{
   shapeType=Capsule
   length=2.758
   radius=1.6566
   shapeName=dyn_Zipper_01_capsule
   center=<2.3723 -0.0238 -0.0247>
   rotation=<rot 0 -90.1515 -89.4855>
  }]
 dyn_jacket_a_01 [{shapeType=Capsule, length=1.9168, radius=9.9119, shapeName=dyn_jacket_a_01_capsule, center=<8.8324 0 6>, rotation=<rot 90 0 0>}]
 dyn_jacket_a_02 [{shapeType=Capsule, length=1.9168, radius=9.9119, shapeName=dyn_jacket_a_02_capsule, center=<8.8324 0 6>, rotation=<rot 90 0 0>}]
 dyn_jacket_a_MSTR [{shapeType=Capsule, length=0.505, radius=0.505, shapeName=dyn_jacket_a_MSTR_capsule, rotation=<rot 0 -90 0>}]
 dyn_jacket_a_constraint [{shapeType=Capsule, length=0.505, radius=0.505, shapeName=dyn_jacket_a_constraint_capsule, rotation=<rot 0 -90 0>}]
 dyn_jacket_b_01 [{shapeType=Capsule, length=1.9168, radius=9.9119, shapeName=dyn_jacket_b_01_capsule, center=<8.5846 0 6>, rotation=<rot 90 0 0>}]
 dyn_jacket_b_02 [{shapeType=Capsule, length=1.9168, radius=9.9119, shapeName=dyn_jacket_b_02_capsule, center=<8.5846 0 6>, rotation=<rot 90 0 0>}]
 dyn_jacket_b_MSTR [{shapeType=Capsule, length=0.505, radius=0.505, shapeName=dyn_jacket_b_MSTR_capsule, rotation=<rot 0 -90 0>}]
 dyn_jacket_b_constraint [{shapeType=Capsule, length=0.505, radius=0.505, shapeName=dyn_jacket_b_constraint_capsule, rotation=<rot 0 -90 0>}]
 dyn_jacket_c_01 [{shapeType=Capsule, length=21.0132, radius=11.6719, shapeName=dyn_jacket_c_01_capsule, center=<8.6494 0 6>, rotation=<rot 0 0 90>}]
 dyn_jacket_c_02 [{shapeType=Capsule, length=21.0132, radius=15.4297, shapeName=dyn_jacket_c_02_capsule, center=<8.6494 0 6>, rotation=<rot 0 0 90>}]
 dyn_jacket_c_MSTR [{shapeType=Capsule, length=0.505, radius=0.505, shapeName=dyn_jacket_c_MSTR_capsule, rotation=<rot 0 -90 0>}]
 dyn_jacket_c_constraint [{shapeType=Capsule, length=0.505, radius=0.505, shapeName=dyn_jacket_c_constraint_capsule, rotation=<rot 0 -90 0>}]
 dyn_jacket_d_01 [
  {shapeType=Capsule, length=24.3675, radius=11.0286, shapeName=dyn_jacket_d_01_capsule, center=<8.9105 0 6>, rotation=<rot 0 0 90>}
  {
   shapeType=Capsule
   length=23.3397
   radius=12.0564
   shapeName=dyn_jacket_d_01_capsule
   center=<14.2369 14.8546 -3.4596>
   rotation=<rot 0 0 123.9156>
  }
 ]
 dyn_jacket_d_02 [{shapeType=Capsule, length=25.568, radius=11.6675, shapeName=dyn_jacket_d_02_capsule, center=<8.9105 0 6>, rotation=<rot 0 0 90>}]
 dyn_jacket_d_MSTR [{shapeType=Capsule, length=0.505, radius=0.505, shapeName=dyn_jacket_d_MSTR_capsule, rotation=<rot 0 -90 0>}]
 dyn_jacket_d_constraint [{shapeType=Capsule, length=0.505, radius=0.505, shapeName=dyn_jacket_d_constraint_capsule, rotation=<rot 0 -90 0>}]
 dyn_jacket_e_01 [{shapeType=Capsule, length=1.2104, radius=12.014, shapeName=dyn_jacket_e_01_capsule, center=<8.5346 0 6>, rotation=<rot 90 0 0>}]
 dyn_jacket_e_02 [{shapeType=Capsule, length=0, radius=15.0637, shapeName=dyn_jacket_e_02_capsule, center=<8.5346 0 6>, rotation=<rot 90 0 0>}]
 dyn_jacket_e_MSTR [{shapeType=Capsule, length=0.505, radius=0.505, shapeName=dyn_jacket_e_MSTR_capsule, rotation=<rot 0 -90 0>}]
 dyn_jacket_e_constraint [{shapeType=Capsule, length=0.505, radius=0.505, shapeName=dyn_jacket_e_constraint_capsule, rotation=<rot 0 -90 0>}]
 dyn_jacket_f_01 [
  {shapeType=Capsule, length=3.1669, radius=8.6618, shapeName=dyn_jacket_f_01_capsule, center=<7.9343 0 6>, rotation=<rot 90 0 0>}
  {shapeType=Capsule, length=8.5586, radius=7.457, shapeName=dyn_jacket_f_01_capsule, center=<8.9258 -7.5645 8>, rotation=<rot 90 0 0>}
 ]
 dyn_jacket_f_02 [
  {shapeType=Capsule, length=4.4944, radius=7.3343, shapeName=dyn_jacket_f_02_capsule, center=<7.9343 0 11>, rotation=<rot 90 0 0>}
  {shapeType=Capsule, length=3.8751, radius=7.9536, shapeName=dyn_jacket_f_02_capsule, center=<6.0373 -7 10>, rotation=<rot 90 0 0>}
 ]
 dyn_jacket_f_MSTR [{shapeType=Capsule, length=0.505, radius=0.505, shapeName=dyn_jacket_f_MSTR_capsule, rotation=<rot 0 -90 0>}]
 dyn_jacket_f_constraint [{shapeType=Capsule, length=0.505, radius=0.505, shapeName=dyn_jacket_f_constraint_capsule, rotation=<rot 0 -90 0>}]
 dyn_jacket_g_01 [{shapeType=Capsule, length=0.2856, radius=11.5431, shapeName=dyn_jacket_g_02_capsule, center=<6.6193 0 6>, rotation=<rot 90 0 0>}]
 dyn_jacket_g_02 [{shapeType=Capsule, length=6.7556, radius=13.2131, shapeName=dyn_jacket_g_02_capsule, center=<4.1375 0 6>, rotation=<rot 90 0 0>}]
 dyn_jacket_g_03 [{shapeType=Capsule, length=0, radius=14.4116, shapeName=dyn_jacket_g_03_capsule, center=<7.4317 0 6>, rotation=<rot 90 0 0>}]
 dyn_jacket_g_MSTR [{shapeType=Capsule, length=0.505, radius=0.505, shapeName=dyn_jacket_g_MSTR_capsule, rotation=<rot 0 -90 0>}]
 dyn_jacket_g_constraint [{shapeType=Capsule, length=0.505, radius=0.505, shapeName=dyn_jacket_g_constraint_capsule, rotation=<rot 0 -90 0>}]
 pelvis [{shapeType=Capsule, length=8.8483, radius=16.3141, shapeName=pelvis_capsule}]
 spine_03 [
  {shapeType=Capsule, length=30.1155, radius=17.0574, shapeName=spine_03_capsule, center=<0 -0.0222 0>, rotation=<rot -90 0 0>}
  {shapeType=Capsule, length=17.4953, radius=15.0288, shapeName=spine_03_capsule, center=<-0.1876 -4.0485 0>, rotation=<rot -90 180 180>}
 ]
 spine_05 [{shapeType=Capsule, length=2, radius=2, shapeName=spine_05_capsule, rotation=<rot 90 0 0>}]
}
constraints [
 {
  bone1Name=spine_05
  bone2Name=spine_03
  swing1LimitDegrees=45
  swing1Motion=Free
  swing2LimitDegrees=45
  swing2Motion=Free
  twistLimitDegrees=45
  twistMotion=Free
 }
 {
  bone1Name=dyn_jacket_a_constraint
  bone2Name=spine_05
  swing1LimitDegrees=45
  swing1Motion=Free
  swing2LimitDegrees=45
  swing2Motion=Free
  twistLimitDegrees=45
  twistMotion=Free
 }
 {
  bone1Name=dyn_jacket_b_constraint
  bone2Name=spine_05
  swing1LimitDegrees=45
  swing1Motion=Free
  swing2LimitDegrees=45
  swing2Motion=Free
  twistLimitDegrees=45
  twistMotion=Free
 }
 {
  bone1Name=dyn_jacket_b_MSTR
  bone2Name=dyn_jacket_b_constraint
  swing1LimitDegrees=45
  swing1Motion=Free
  swing2LimitDegrees=45
  swing2Motion=Free
  twistLimitDegrees=45
  twistMotion=Free
 }
 {
  bone1Name=dyn_jacket_a_MSTR
  bone2Name=dyn_jacket_a_constraint
  swing1LimitDegrees=45
  swing1Motion=Free
  swing2LimitDegrees=45
  swing2Motion=Free
  twistLimitDegrees=45
  twistMotion=Free
 }
 {
  bone1Name=dyn_jacket_c_constraint
  bone2Name=spine_05
  swing1LimitDegrees=45
  swing1Motion=Free
  swing2LimitDegrees=45
  swing2Motion=Free
  twistLimitDegrees=45
  twistMotion=Free
 }
 {
  bone1Name=dyn_jacket_c_MSTR
  bone2Name=dyn_jacket_c_constraint
  swing1LimitDegrees=45
  swing1Motion=Free
  swing2LimitDegrees=45
  swing2Motion=Free
  twistLimitDegrees=45
  twistMotion=Free
 }
 {
  bone1Name=dyn_jacket_d_constraint
  bone2Name=spine_05
  swing1LimitDegrees=45
  swing1Motion=Free
  swing2LimitDegrees=45
  swing2Motion=Free
  twistLimitDegrees=45
  twistMotion=Free
 }
 {
  bone1Name=dyn_jacket_d_MSTR
  bone2Name=dyn_jacket_d_constraint
  swing1LimitDegrees=45
  swing1Motion=Free
  swing2LimitDegrees=45
  swing2Motion=Free
  twistLimitDegrees=45
  twistMotion=Free
 }
 {
  bone1Name=dyn_jacket_e_constraint
  bone2Name=spine_05
  swing1LimitDegrees=45
  swing1Motion=Free
  swing2LimitDegrees=45
  swing2Motion=Free
  twistLimitDegrees=45
  twistMotion=Free
 }
 {
  bone1Name=dyn_jacket_e_MSTR
  bone2Name=dyn_jacket_e_constraint
  swing1LimitDegrees=45
  swing1Motion=Free
  swing2LimitDegrees=45
  swing2Motion=Free
  twistLimitDegrees=45
  twistMotion=Free
 }
 {
  bone1Name=dyn_jacket_f_constraint
  bone2Name=spine_05
  swing1LimitDegrees=45
  swing1Motion=Free
  swing2LimitDegrees=45
  swing2Motion=Free
  twistLimitDegrees=45
  twistMotion=Free
 }
 {
  bone1Name=dyn_jacket_f_MSTR
  bone2Name=dyn_jacket_f_constraint
  swing1LimitDegrees=45
  swing1Motion=Free
  swing2LimitDegrees=45
  swing2Motion=Free
  twistLimitDegrees=45
  twistMotion=Free
 }
 {
  bone1Name=dyn_jacket_g_constraint
  bone2Name=spine_05
  swing1LimitDegrees=45
  swing1Motion=Free
  swing2LimitDegrees=45
  swing2Motion=Free
  twistLimitDegrees=45
  twistMotion=Free
 }
 {
  bone1Name=dyn_jacket_g_MSTR
  bone2Name=dyn_jacket_g_constraint
  swing1LimitDegrees=45
  swing1Motion=Free
  swing2LimitDegrees=45
  swing2Motion=Free
  twistLimitDegrees=45
  twistMotion=Free
 }
 {bone1Name=dyn_jacket_a_01, bone2Name=dyn_jacket_a_MSTR, swing1LimitDegrees=20, swing2LimitDegrees=45, twistLimitDegrees=20}
 {bone1Name=dyn_jacket_a_02, bone2Name=dyn_jacket_a_01, swing1LimitDegrees=20, swing2LimitDegrees=45, twistLimitDegrees=20}
 {bone1Name=dyn_jacket_b_01, bone2Name=dyn_jacket_b_MSTR, swing1LimitDegrees=20, swing2LimitDegrees=45, twistLimitDegrees=20}
 {bone1Name=dyn_jacket_b_02, bone2Name=dyn_jacket_b_01, swing1LimitDegrees=20, swing2LimitDegrees=45, twistLimitDegrees=20}
 {bone1Name=dyn_jacket_c_01, bone2Name=dyn_jacket_c_MSTR, swing1LimitDegrees=20, swing2LimitDegrees=45, twistLimitDegrees=20}
 {bone1Name=dyn_jacket_c_02, bone2Name=dyn_jacket_c_01, swing1LimitDegrees=20, swing2LimitDegrees=45, twistLimitDegrees=20}
 {bone1Name=dyn_jacket_d_01, bone2Name=dyn_jacket_d_MSTR, swing1LimitDegrees=20, swing2LimitDegrees=45, twistLimitDegrees=20}
 {bone1Name=dyn_jacket_d_02, bone2Name=dyn_jacket_d_01, swing1LimitDegrees=20, swing2LimitDegrees=45, twistLimitDegrees=20}
 {bone1Name=dyn_jacket_e_01, bone2Name=dyn_jacket_e_MSTR, swing1LimitDegrees=20, swing2LimitDegrees=45, twistLimitDegrees=20}
 {bone1Name=dyn_jacket_e_02, bone2Name=dyn_jacket_e_01, swing1LimitDegrees=20, swing2LimitDegrees=45, twistLimitDegrees=20}
 {bone1Name=dyn_jacket_f_01, bone2Name=dyn_jacket_f_MSTR, swing1LimitDegrees=20, swing2LimitDegrees=45, twistLimitDegrees=20}
 {bone1Name=dyn_jacket_f_02, bone2Name=dyn_jacket_f_01, swing1LimitDegrees=20, swing2LimitDegrees=45, twistLimitDegrees=20}
 {bone1Name=dyn_jacket_g_02, bone2Name=dyn_jacket_g_01, swing1LimitDegrees=20, swing2LimitDegrees=45, twistLimitDegrees=20}
 {bone1Name=dyn_jacket_g_03, bone2Name=dyn_jacket_g_02, swing1LimitDegrees=20, swing2LimitDegrees=45, twistLimitDegrees=20}
 {bone1Name=dyn_Zipper_01, bone2Name=dyn_jacket_a_02, swing1LimitDegrees=20, swing2LimitDegrees=45, twistLimitDegrees=20}
 {bone1Name=dyn_jacket_g_01, bone2Name=dyn_jacket_g_MSTR, swing1LimitDegrees=20, swing2LimitDegrees=45, twistLimitDegrees=20}
 {
  bone1Name=spine_03
  bone2Name=pelvis
  swing1LimitDegrees=45
  swing1Motion=Free
  swing2LimitDegrees=45
  swing2Motion=Free
  twistLimitDegrees=45
  twistMotion=Free
 }
]
