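# Post-process style AnimBlueprint for the BananaAdventure male body cosmetic, adding secondary motion and corrective offsets to jacket, scarf, map, shorts, buckle, zipper and bag bones.
# Two RigidBody sims, pelvis- and spine_03-based, branch from cached PreJacketSim and blend on emote, skydive, parachute or melee state.
# AnimDynamics then swings scarf and map.
# Weighted constraints, calf-driven Flap nodes and toggleable additive bone transforms fit shorts and accessories to the pose.
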AnimBlueprint /BRCosmetics/Characters/Player/Male/Medium/Bodies/M_MED_BananaAdventure/Meshes/M_MED_BananaAdventure_AnimBP
target_skeleton=&/M_MED_BananaAdventure_Skeleton
graphs [
 AnimGraph {nodes [
   n1 = "Transform (Modify) Bone - Bone: deform_neckline_r_02_root"!(ComponentPose, bAlphaBoolEnabled:bool=True, Translation="7.418,1.116,-8.749", Rotation="80,0.329,0.232", Scale="1,1,1") -> (Pose) {
    alphaInputType=Bool
    rotationMode=BMM_Additive
    rotationSpace=BCS_BoneSpace
    scaleMode=BMM_Additive
    scaleSpace=BCS_BoneSpace
    translationMode=BMM_Additive
    translationSpace=BCS_BoneSpace
    alphaBoolBlend {blendInTime=0.3, blendOption=Linear, blendOutTime=1}
    boneToModify {boneName=deform_neckline_r_02_root}
    rotation=<rot 80 0.329 0.232>
    translation=<7.418 1.116 -8.749>
   }
   n2 = "Transform (Modify) Bone - Bone: deform_neckline_r_03_root"!(ComponentPose=@n1, bAlphaBoolEnabled:bool=True, Translation="1.627,0,-3.26", Rotation="0,0.414,3.266", Scale="1,1,1") -> (Pose) {
    alphaInputType=Bool
    rotationMode=BMM_Additive
    rotationSpace=BCS_BoneSpace
    scaleMode=BMM_Additive
    scaleSpace=BCS_BoneSpace
    translationMode=BMM_Additive
    translationSpace=BCS_BoneSpace
    alphaBoolBlend {blendInTime=0.3, blendOption=Linear, blendOutTime=1}
    boneToModify {boneName=deform_neckline_r_03_root}
    rotation {roll=3.266, yaw=0.414}
    translation {x=1.627, z=-3.26}
   }
   n3 = "Transform (Modify) Bone - Bone: deform_neckline_rear_root"!(ComponentPose=@n2, bAlphaBoolEnabled:bool=True, Translation="5.920,1.591,-3.247", Rotation="-32.359,-1.023,3.812", Scale="1,1,1") -> (Pose) {
    alphaInputType=Bool
    rotationMode=BMM_Additive
    rotationSpace=BCS_BoneSpace
    scaleMode=BMM_Additive
    scaleSpace=BCS_BoneSpace
    translationMode=BMM_Additive
    translationSpace=BCS_BoneSpace
    alphaBoolBlend {blendInTime=0.3, blendOption=Linear, blendOutTime=1}
    boneToModify {boneName=deform_neckline_rear_root}
    rotation=<rot -32.359 -1.023 3.812>
    translation=<5.92 1.591 -3.247>
   }
   n4 = "Transform (Modify) Bone - Bone: deform_neckline_r_01_root"!(ComponentPose=@n3, bAlphaBoolEnabled:bool=True, Translation="-4.135,0.000,0.274", Rotation="10.091,2.543,19.188", Scale="1,1,1") -> (Pose) {
    alphaInputType=Bool
    rotationMode=BMM_Additive
    rotationSpace=BCS_BoneSpace
    scaleMode=BMM_Additive
    scaleSpace=BCS_BoneSpace
    translationMode=BMM_Additive
    translationSpace=BCS_BoneSpace
    alphaBoolBlend {blendInTime=0.3, blendOption=Linear, blendOutTime=1}
    boneToModify {boneName=deform_neckline_r_01_root}
    rotation=<rot 10.091 2.543 19.188>
    translation {x=-4.135, z=0.274}
   }
   n5 = "Transform (Modify) Bone - Bone: deform_neckline_l_02_root"!(ComponentPose=@n4, bAlphaBoolEnabled:bool=True, Translation="0,0,0", Rotation="0,0,0", Scale="1,1,1") -> (Pose) {
    alphaInputType=Bool
    rotationMode=BMM_Additive
    rotationSpace=BCS_BoneSpace
    scaleMode=BMM_Additive
    scaleSpace=BCS_BoneSpace
    translationMode=BMM_Additive
    translationSpace=BCS_BoneSpace
    alphaBoolBlend {blendInTime=0.3, blendOption=Linear, blendOutTime=1}
    boneToModify {boneName=deform_neckline_l_02_root}
   }
   "Transform (Modify) Bone - Bone: deform_neckline_l_03_root"!(ComponentPose=@n5, bAlphaBoolEnabled:bool=True, Translation="0,0,0", Rotation="0,0,0", Scale="1,1,1") -> (Pose) {
    alphaInputType=Bool
    rotationMode=BMM_Additive
    rotationSpace=BCS_BoneSpace
    scaleMode=BMM_Additive
    scaleSpace=BCS_BoneSpace
    translationMode=BMM_Additive
    translationSpace=BCS_BoneSpace
    alphaBoolBlend {blendInTime=0.3, blendOption=Linear, blendOutTime=1}
    boneToModify {boneName=deform_neckline_l_03_root}
   }
   n6 = "Transform (Modify) Bone - Bone: dyn_shorts_bk_l"(Alpha=1, Translation="7.751,0.119,-0.000", Rotation="3.505,-0.000,-0.000", Scale="1,1,1") {
    rotationMode=BMM_Additive
    rotationSpace=BCS_BoneSpace
    scaleMode=BMM_Additive
    scaleSpace=BCS_BoneSpace
    translationMode=BMM_Additive
    translationSpace=BCS_BoneSpace
    alphaBoolBlend {blendInTime=0.2, blendOption=Linear, blendOutTime=0.4}
    alphaScaleBiasClamp {bInterpResult=true}
    boneToModify {boneName=dyn_shorts_bk_l}
    rotation {pitch=3.505}
    translation=<7.751 0.119>
   }
   n7 = "Transform (Modify) Bone - Bone: dyn_shorts_bk_r"(ComponentPose=@n6, Alpha=1, Translation="-3.551,-2.419,3.525", Rotation="-11.037,0.041,-0.043", Scale="1,1,1") {
    rotationMode=BMM_Additive
    rotationSpace=BCS_BoneSpace
    scaleMode=BMM_Additive
    scaleSpace=BCS_BoneSpace
    translationMode=BMM_Additive
    translationSpace=BCS_BoneSpace
    alphaBoolBlend {blendInTime=0.2, blendOption=Linear, blendOutTime=0.4}
    alphaScaleBiasClamp {bInterpResult=true}
    boneToModify {boneName=dyn_shorts_bk_r}
    rotation=<rot -11.037 0.041 -0.043>
    translation=<-3.551 -2.419 3.525>
   }
   "Transform (Modify) Bone - Bone: dyn_shorts_out_r"(ComponentPose=@n7, Alpha=1, Translation="-3.273,-1.492,0.856", Rotation="-5.748,0.155,8.504", Scale="1,1,1") {
    rotationMode=BMM_Additive
    rotationSpace=BCS_BoneSpace
    scaleMode=BMM_Additive
    scaleSpace=BCS_BoneSpace
    translationMode=BMM_Additive
    translationSpace=BCS_BoneSpace
    alphaBoolBlend {blendInTime=0.2, blendOption=Linear, blendOutTime=0.4}
    alphaScaleBiasClamp {bInterpResult=true}
    boneToModify {boneName=dyn_shorts_out_r}
    rotation=<rot -5.748 0.155 8.504>
    translation=<-3.273 -1.492 0.856>
   }
   n8 = RigidBody!(ComponentPose=Local To Component(LocalPose=Use cached pose 'PreJacketSim'()), Alpha, OverrideWorldGravity:Vector="0,0,-980", ExternalForce) -> (Pose) {
    bClampLinearTranslationLimitToRefPose=true
    bOverrideWorldGravity=true
    overridePhysicsAsset=&/M_MED_BananaAdventure_Physics
    simulationSpace=BaseBoneSpace
    alphaBoolBlend {blendOption=Linear}
    baseBoneRef {boneName=pelvis}
    componentAppliedLinearAccClamp=<5000 5000 5000>
    componentLinearAccScale=<0.4 0.4 0.5>
    overrideWorldGravity {z=-980}
    simSpaceSettings {
     maxAngularAcceleration=1000
     maxAngularVelocity=3
     maxLinearAcceleration=700
     maxLinearVelocity=500
     velocityScaleZ=0
     worldAlpha=0.45
     externalLinearDragV=<0.2 0.2>
    }
   }
   n9 = Component To Local(ComponentPose=@n8)
   n10 = RigidBody(ComponentPose=Local To Component(LocalPose=Use cached pose 'PreJacketSim'()), OverrideWorldGravity="0,0,-980") {
    bClampLinearTranslationLimitToRefPose=true
    bOverrideWorldGravity=true
    overridePhysicsAsset=&M_MED_BananaAdventure_Physics
    simulationSpace=BaseBoneSpace
    alphaBoolBlend {blendOption=Linear}
    baseBoneRef {boneName=spine_03}
    componentAppliedLinearAccClamp=<5000 5000 5000>
    componentLinearAccScale=<0.35 0.35 0.45>
    overrideWorldGravity {z=-980}
    simSpaceSettings {
     maxAngularAcceleration=1000
     maxAngularVelocity=3
     maxLinearAcceleration=700
     maxLinearVelocity=500
     velocityScaleZ=0
     worldAlpha=0.3
     externalLinearDragV=<0.2 0.2>
    }
   }
   n11 = "Transform (Modify) Bone - Bone: dyn_jacket_g_constraint"!(ComponentPose=@n10, bAlphaBoolEnabled:bool=True, Translation="-1.884,0.085,1.5", Rotation="4.244,-1.540,-19.943", Scale="1,1,1") -> (Pose) {
    alphaInputType=Bool
    rotationMode=BMM_Additive
    rotationSpace=BCS_BoneSpace
    scaleMode=BMM_Additive
    scaleSpace=BCS_BoneSpace
    translationMode=BMM_Additive
    translationSpace=BCS_BoneSpace
    alphaBoolBlend {blendInTime=0.1, blendOption=Linear, blendOutTime=0.6}
    boneToModify {boneName=dyn_jacket_g_constraint}
    rotation=<rot 4.244 -1.54 -19.943>
    translation=<-1.884 0.085 1.5>
   }
   n12 = "Transform (Modify) Bone - Bone: dyn_jacket_a_constraint"!(ComponentPose=@n11, bAlphaBoolEnabled:bool=True, Translation="-1.884,0.085,1.5", Rotation="2.601,15.735,-30.190", Scale="1,1,1") -> (Pose) {
    alphaInputType=Bool
    rotationMode=BMM_Additive
    rotationSpace=BCS_BoneSpace
    scaleMode=BMM_Additive
    scaleSpace=BCS_BoneSpace
    translationMode=BMM_Additive
    translationSpace=BCS_BoneSpace
    alphaBoolBlend {blendInTime=0.1, blendOption=Linear, blendOutTime=0.6}
    boneToModify {boneName=dyn_jacket_a_constraint}
    rotation=<rot 2.601 15.735 -30.19>
    translation=<-1.884 0.085 1.5>
   }
   n13 = "Transform (Modify) Bone - Bone: dyn_jacket_g_constraint"(ComponentPose=@n12, bAlphaBoolEnabled=True, Translation="-4.795,3.549,0.174", Rotation="-4.790,0.885,-10.471", Scale="1,1,1") {
    alphaInputType=Bool
    rotationMode=BMM_Additive
    rotationSpace=BCS_BoneSpace
    scaleMode=BMM_Additive
    scaleSpace=BCS_BoneSpace
    translationMode=BMM_Additive
    translationSpace=BCS_BoneSpace
    alphaBoolBlend {blendInTime=0.1, blendOption=Linear, blendOutTime=0.6}
    boneToModify {boneName=dyn_jacket_g_constraint}
    rotation=<rot -4.79 0.885 -10.471>
    translation=<-4.795 3.549 0.174>
   }
   n14 = "Transform (Modify) Bone - Bone: dyn_jacket_f_constraint"!(ComponentPose=@n13, bAlphaBoolEnabled:bool=True, Translation="-3.251,0.000,-0.000", Rotation="0,0,0", Scale="1,1,1") -> (Pose) {
    alphaInputType=Bool
    rotationMode=BMM_Additive
    rotationSpace=BCS_BoneSpace
    scaleMode=BMM_Additive
    scaleSpace=BCS_BoneSpace
    translationMode=BMM_Additive
    translationSpace=BCS_BoneSpace
    alphaBoolBlend {blendInTime=0.1, blendOption=Linear, blendOutTime=0.6}
    boneToModify {boneName=dyn_jacket_f_constraint}
    translation {x=-3.251}
   }
   n15 = "Transform (Modify) Bone - Bone: dyn_jacket_e_constraint"!(ComponentPose=@n14, bAlphaBoolEnabled:bool=True, Translation="-3.501,0.000,0.000", Rotation="0,0,0", Scale="1,1,1") -> (Pose) {
    alphaInputType=Bool
    rotationMode=BMM_Additive
    rotationSpace=BCS_BoneSpace
    scaleMode=BMM_Additive
    scaleSpace=BCS_BoneSpace
    translationMode=BMM_Additive
    translationSpace=BCS_BoneSpace
    alphaBoolBlend {blendInTime=0.1, blendOption=Linear, blendOutTime=0.6}
    boneToModify {boneName=dyn_jacket_e_constraint}
    translation {x=-3.501}
   }
   n16 = "Transform (Modify) Bone - Bone: dyn_jacket_a_constraint"(ComponentPose=@n15, bAlphaBoolEnabled=True, Translation="-4.740,7.247,1.492", Rotation="-4.79,0.885,0", Scale="1,1,1") {
    alphaInputType=Bool
    rotationMode=BMM_Additive
    rotationSpace=BCS_BoneSpace
    scaleMode=BMM_Additive
    scaleSpace=BCS_BoneSpace
    translationMode=BMM_Additive
    translationSpace=BCS_BoneSpace
    alphaBoolBlend {blendInTime=0.1, blendOption=Linear, blendOutTime=0.6}
    boneToModify {boneName=dyn_jacket_a_constraint}
    rotation {pitch=-4.79, yaw=0.885}
    translation=<-4.74 7.247 1.492>
   }
   n17 = "Transform (Modify) Bone - Bone: dyn_jacket_b_constraint"!(ComponentPose=@n16, bAlphaBoolEnabled:bool=True, Translation="2.257,-0.001,-0.000", Rotation="7.491,0.000,-0.000", Scale="1,1,1") -> (Pose) {
    alphaInputType=Bool
    rotationMode=BMM_Additive
    rotationSpace=BCS_BoneSpace
    scaleMode=BMM_Additive
    scaleSpace=BCS_BoneSpace
    translationMode=BMM_Additive
    translationSpace=BCS_BoneSpace
    alphaBoolBlend {blendInTime=0.1, blendOption=Linear, blendOutTime=0.6}
    boneToModify {boneName=dyn_jacket_b_constraint}
    rotation {pitch=7.491}
    translation=<2.257 -0.001>
   }
   n18 = "Transform (Modify) Bone - Bone: dyn_jacket_c_constraint"!(ComponentPose=@n17, bAlphaBoolEnabled:bool=True, Translation="1.571,-1.036,0.000", Rotation="5.844,-0.363,-3.563", Scale="1,1,1") -> (Pose) {
    alphaInputType=Bool
    rotationMode=BMM_Additive
    rotationSpace=BCS_BoneSpace
    scaleMode=BMM_Additive
    scaleSpace=BCS_BoneSpace
    translationMode=BMM_Additive
    translationSpace=BCS_BoneSpace
    alphaBoolBlend {blendInTime=0.1, blendOption=Linear, blendOutTime=0.6}
    boneToModify {boneName=dyn_jacket_c_constraint}
    rotation=<rot 5.844 -0.363 -3.563>
    translation=<1.571 -1.036>
   }
   n19 = Component To Local(ComponentPose=@n18)
   n20 = OR Boolean!(A=Get bIsPlayingEmote(), B=Get bIsSkyDiving(), C:bool=Get bIsParachuteOpened(), D:bool=Get bIsPlayingMeleeAnim()) -> (ReturnValue)
   n21 = Anim Dynamics - Bone: dyn_scarf_01(
    ComponentPose=Local To Component(LocalPose=Blend Poses by bool(BlendPose_0=@n19, BlendPose_1=@n9, BlendTime_0=0.5, BlendTime_1=0.5, bActiveValue=@n20) {blendType=Linear})
    Alpha=0.7
    ExternalForce="0,0,0"
   ) {
    angularDampingOverride=1
    angularSpringConstant=500
    bAngularSpring=true
    bOverrideAngularDamping=true
    bUseGravityOverride=true
    gravityScale=0.35
    simulationSpace=RootRelative
    alphaBoolBlend {blendOption=Linear}
    boundBone {boneName=dyn_scarf_01}
    componentAppliedLinearAccClamp=<5000 5000 5000>
    componentLinearAccScale=<0.6 0.8 0.6>
    componentLinearVelScale=<1 1.2 1>
    gravityOverride {z=-1500}
    physicsBodyDefinitions [{
      boundBone {boneName=dyn_scarf_01}
      boxExtents=<12 23 23>
      constraintSetup {angularLimitsMax=<1 -3 20>, angularLimitsMin=<-1 -45 -20>, angularTarget=<1 0 0>}
      localJointOffset=<3 0 0>
     }]
   }
   n22 = Anim Dynamics - Bone: dyn_scarf_02(ComponentPose=@n21, Alpha=0.7, ExternalForce="0,0,0") {
    angularDampingOverride=1
    angularSpringConstant=100
    bAngularSpring=true
    bOverrideAngularDamping=true
    bUseGravityOverride=true
    gravityScale=0.35
    simulationSpace=RootRelative
    alphaBoolBlend {blendOption=Linear}
    boundBone {boneName=dyn_scarf_02}
    componentAppliedLinearAccClamp=<5000 5000 5000>
    componentLinearAccScale=<0.6 0.8 0.6>
    componentLinearVelScale=<1 1.2 1>
    gravityOverride {z=-1500}
    physicsBodyDefinitions [{
      boundBone {boneName=dyn_scarf_02}
      boxExtents=<12 23 23>
      constraintSetup {angularLimitsMax=<1 -3 20>, angularLimitsMin=<-1 -45 -20>, angularTarget=<1 0 0>}
      localJointOffset=<3 0 0>
     }]
   }
   n23 = Anim Dynamics - Bone: dyn_scarf_03(ComponentPose=@n22, Alpha=0.65, ExternalForce="0,0,0") {
    angularDampingOverride=1
    angularSpringConstant=100
    bAngularSpring=true
    bOverrideAngularDamping=true
    bUseGravityOverride=true
    gravityScale=0.35
    simulationSpace=RootRelative
    alphaBoolBlend {blendOption=Linear}
    boundBone {boneName=dyn_scarf_03}
    componentAppliedLinearAccClamp=<5000 5000 5000>
    componentLinearAccScale=<0.6 0.8 0.6>
    componentLinearVelScale=<1 1.2 1>
    gravityOverride {z=-1500}
    physicsBodyDefinitions [{
      boundBone {boneName=dyn_scarf_03}
      boxExtents=<12 23 23>
      constraintSetup {angularLimitsMax=<1 -3 45>, angularLimitsMin=<-1 -45 -45>, angularTarget=<1 0 0>}
      localJointOffset=<3 0 0>
     }]
   }
   n24 = Anim Dynamics - Bone: dyn_map(ComponentPose=@n23, Alpha=0.05, ExternalForce="0,0,0") {
    angularDampingOverride=1
    bOverrideAngularDamping=true
    bUseGravityOverride=true
    gravityScale=0.35
    simulationSpace=RootRelative
    alphaBoolBlend {blendOption=Linear}
    boundBone {boneName=dyn_map}
    componentAppliedLinearAccClamp=<5000 5000 5000>
    gravityOverride {z=-800}
    physicsBodyDefinitions [{
      boundBone {boneName=dyn_map}
      boxExtents=<5 10 10>
      constraintSetup {angularLimitsMax=<10 45 20>, angularLimitsMin=<-10 -45 -20>, angularTarget=<1 0 0>}
      localJointOffset=<4 0 0>
     }]
   }
   n25 = Anim Dynamics - Bone: dyn_mapFront(ComponentPose=@n24, Alpha=0.1, ExternalForce="0,0,0") {
    angularDampingOverride=0.9999
    bOverrideAngularDamping=true
    bUseGravityOverride=true
    gravityScale=0.35
    simulationSpace=RootRelative
    alphaBoolBlend {blendOption=Linear}
    boundBone {boneName=dyn_mapFront}
    componentAppliedLinearAccClamp=<5000 5000 5000>
    gravityOverride {z=-800}
    physicsBodyDefinitions [{
      boundBone {boneName=dyn_mapFront}
      boxExtents=<5 10 10>
      constraintSetup {angularLimitsMax=<10 45 20>, angularLimitsMin=<-10 -45 -20>, angularTarget=<1 0 0>}
      localJointOffset=<4 0 0>
     }]
   }
   n26 = Anim Dynamics - Bone: dyn_mapBack(ComponentPose=@n25, Alpha=0.12, ExternalForce="0,0,0") {
    angularDampingOverride=0.9999
    bOverrideAngularDamping=true
    bUseGravityOverride=true
    gravityScale=0.35
    simulationSpace=RootRelative
    alphaBoolBlend {blendOption=Linear}
    boundBone {boneName=dyn_mapBack}
    componentAppliedLinearAccClamp=<5000 5000 5000>
    gravityOverride {z=-800}
    physicsBodyDefinitions [{
      boundBone {boneName=dyn_mapBack}
      boxExtents=<5 10 10>
      constraintSetup {angularLimitsMax=<10 45 20>, angularLimitsMin=<-10 -45 -20>, angularTarget=<1 0 0>}
      localJointOffset=<4 0 0>
     }]
   }
   n27 = "Constraint - dyn_buckle_MSTR (PPPPPPPP)"!(
    ComponentPose=@n26
    Alpha=1
    ConstraintWeights_0=0.4649
    ConstraintWeights_1:float=0.0216
    ConstraintWeights_2:float=0.0392
    ConstraintWeights_3:float=0.0255
    ConstraintWeights_4:float=0.1476
    ConstraintWeights_5:float=0.0934
    ConstraintWeights_6:float=0.1079
    ConstraintWeights_7:float=0.0998
   ) -> (Pose) {
    alphaBoolBlend {blendOption=Linear}
    boneToModify {boneName=dyn_buckle_MSTR}
    constraintSetup [
     {targetBone {boneName=pelvis}}
     {targetBone {boneName=spine_01}}
     {targetBone {boneName=spine_02}}
     {targetBone {boneName=spine_03}}
     {targetBone {boneName=thigh_twist_01_l}}
     {targetBone {boneName=deform_groin_l}}
     {targetBone {boneName=thigh_twist_01_r}}
     {targetBone {boneName=deform_groin_r}}
    ]
    constraintWeights [1,1,1,1,1,1,1,1]
   }
   n28 = "Constraint - dyn_buckle_01 (RR)"!(ComponentPose=@n27, bAlphaBoolEnabled:bool=True, ConstraintWeights_0=0.5, ConstraintWeights_1:float=0.5) -> (Pose) {
    alphaInputType=Bool
    alphaBoolBlend {blendInTime=0.1, blendOption=Linear, blendOutTime=0.1}
    boneToModify {boneName=dyn_buckle_01}
    constraintSetup [{transformType=Rotation, targetBone {boneName=thigh_l}}, {transformType=Rotation, targetBone {boneName=thigh_r}}]
    constraintWeights [1,1]
   }
   n29 = "Constraint - dyn_Zipper_MSTR (PPPPPPPP)"!(
    ComponentPose=@n28
    Alpha=1
    ConstraintWeights_0=0.0013
    ConstraintWeights_1:float=0.0578
    ConstraintWeights_2:float=0.7724
    ConstraintWeights_3:float=0.0072
    ConstraintWeights_4:float=0.0031
    ConstraintWeights_5:float=0.1045
    ConstraintWeights_6:float=0.052
    ConstraintWeights_7:float=0.0017
   ) -> (Pose) {
    alphaBoolBlend {blendOption=Linear}
    boneToModify {boneName=dyn_Zipper_MSTR}
    constraintSetup [
     {targetBone {boneName=deform_pec_l}}
     {targetBone {boneName=dyn_jacket_a_01}}
     {targetBone {boneName=dyn_jacket_a_02}}
     {targetBone {boneName=dyn_jacket_c_01}}
     {targetBone {boneName=dyn_jacket_c_02}}
     {targetBone {boneName=dyn_jacket_b_01}}
     {targetBone {boneName=dyn_jacket_b_02}}
     {targetBone {boneName=dfrm_clavicle_l}}
    ]
    constraintWeights [1,1,1,1,1,1,1,1]
   }
   n30 = "Transform (Modify) Bone - Bone: dyn_buckle_MSTR"!(ComponentPose=@n29, bAlphaBoolEnabled:bool=True, Translation="2.468,1.488,0.109", Rotation="-21.401,25.536,2.967", Scale="1,1,1") -> (Pose) {
    alphaInputType=Bool
    rotationMode=BMM_Additive
    rotationSpace=BCS_BoneSpace
    scaleMode=BMM_Additive
    scaleSpace=BCS_BoneSpace
    translationMode=BMM_Additive
    translationSpace=BCS_BoneSpace
    alphaBoolBlend {blendInTime=0.3, blendOption=Linear, blendOutTime=1}
    boneToModify {boneName=dyn_buckle_MSTR}
    rotation=<rot -21.401 25.536 2.967>
    translation=<2.468 1.488 0.109>
   }
   n31 = Flap Node - Bone: dyn_shorts_bk_l!(ComponentPose:Component Space Pose Link Structure=@n30, Alpha:float=1) -> (Pose:Component Space Pose Link Structure) {
    is=title FortAnimGraphNode_Flap
    targetComponent=RotationY
    alphaBoolBlend {blendOption=Linear}
    alphaScaleBiasClamp {bInterpResult=true, interpSpeedDecreasing=0.1}
    sourceDrives [{
      bUseRange=true
      drivingCurve=null
      multiplier=0.5
      rangeMax=-0.53
      rangeMin=-2.094
      remappedMax=0
      remappedMin=-1.919
      sourceComponent=RotationZ
      useQuaternion=true
      sourceBone {boneName=calf_l}
     }]
    targetBone {boneName=dyn_shorts_bk_l}
   }
   n32 = Flap Node - Bone: dyn_shorts_bk_l(ComponentPose=@n31, Alpha=1) {
    targetComponent=ScaleZ
    alphaBoolBlend {blendOption=Linear}
    alphaScaleBiasClamp {bInterpResult=true, interpSpeedDecreasing=0.1}
    sourceDrives [{
      bUseRange=true
      drivingCurve=null
      multiplier=0.5
      rangeMax=-0.53
      rangeMin=-2.094
      remappedMax=0
      remappedMin=-0.8
      sourceComponent=RotationZ
      useQuaternion=true
      sourceBone {boneName=calf_l}
     }]
    targetBone {boneName=dyn_shorts_bk_l}
   }
   n33 = Flap Node - Bone: dyn_shorts_bk_r!(ComponentPose:Component Space Pose Link Structure=@n32, Alpha:float=1) -> (Pose:Component Space Pose Link Structure) {
    is=title FortAnimGraphNode_Flap
    targetComponent=RotationY
    alphaBoolBlend {blendOption=Linear}
    alphaScaleBiasClamp {bInterpResult=true, interpSpeedDecreasing=0.1}
    sourceDrives [{
      bUseRange=true
      drivingCurve=null
      multiplier=0.5
      rangeMax=-0.53
      rangeMin=-2.094
      remappedMax=0
      remappedMin=-1.99
      sourceComponent=RotationZ
      useQuaternion=true
      sourceBone {boneName=calf_r}
     }]
    targetBone {boneName=dyn_shorts_bk_r}
   }
   n34 = Flap Node - Bone: dyn_shorts_bk_r(ComponentPose=@n33, Alpha=1) {
    targetComponent=ScaleZ
    alphaBoolBlend {blendOption=Linear}
    alphaScaleBiasClamp {bInterpResult=true, interpSpeedDecreasing=0.1}
    sourceDrives [{
      bUseRange=true
      drivingCurve=null
      multiplier=0.5
      rangeMax=-0.53
      rangeMin=-2.094
      remappedMax=0
      remappedMin=-0.9
      sourceComponent=RotationZ
      useQuaternion=true
      sourceBone {boneName=calf_r}
     }]
    targetBone {boneName=dyn_shorts_bk_r}
   }
   n35 = "Constraint - dyn_shorts_bk_r (PPPPPPP)"!(
    ComponentPose=@n34
    Alpha=1
    ConstraintWeights_0=0.1396
    ConstraintWeights_1:float=0.391
    ConstraintWeights_2:float=0.0058
    ConstraintWeights_3:float=0.0944
    ConstraintWeights_4:float=0.0577
    ConstraintWeights_5:float=0.3036
    ConstraintWeights_6:float=0.0078
   ) -> (Pose) {
    alphaBoolBlend {blendInTime=10, blendOption=Linear, blendOutTime=10}
    alphaScaleBiasClamp {bInterpResult=true, interpSpeedDecreasing=0.1}
    boneToModify {boneName=dyn_shorts_bk_r}
    constraintSetup [
     {targetBone {boneName=thigh_r}}
     {targetBone {boneName=calf_r}}
     {targetBone {boneName=calf_twist_01_r}}
     {targetBone {boneName=calf_twist_02_r}}
     {targetBone {boneName=deform_knee_out_r}}
     {targetBone {boneName=deform_knee_in_r}}
     {targetBone {boneName=thigh_twist_01_r}}
    ]
    constraintWeights [1,1,1,1,1,1,1]
   }
   n36 = "Constraint - dyn_shorts_bk_r (R)"(ComponentPose=@n35, Alpha=1, ConstraintWeights_0=1) {
    alphaBoolBlend {blendInTime=0.3, blendOption=Linear, blendOutTime=0.3}
    alphaScaleBiasClamp {bInterpResult=true}
    boneToModify {boneName=dyn_shorts_bk_r}
    constraintSetup [{transformType=Rotation, targetBone {boneName=calf_r}}]
    constraintWeights [1]
   }
   n37 = "Constraint - dyn_shorts_bk_l (R)"(ComponentPose=@n36, Alpha=1, ConstraintWeights_0=1) {
    alphaBoolBlend {blendInTime=0.3, blendOption=Linear, blendOutTime=0.3}
    alphaScaleBiasClamp {bInterpResult=true}
    boneToModify {boneName=dyn_shorts_bk_l}
    constraintSetup [{transformType=Rotation, targetBone {boneName=calf_l}}]
    constraintWeights [1]
   }
   n38 = "Constraint - dyn_shorts_in_r (R)"(ComponentPose=@n37, Alpha=0.5, ConstraintWeights_0=1) {
    alphaBoolBlend {blendInTime=0.3, blendOption=Linear, blendOutTime=0.3}
    alphaScaleBiasClamp {bInterpResult=true}
    boneToModify {boneName=dyn_shorts_in_r}
    constraintSetup [{transformType=Rotation, targetBone {boneName=calf_r}}]
    constraintWeights [1]
   }
   n39 = "Constraint - dyn_shorts_in_l (R)"(ComponentPose=@n38, Alpha=0.5, ConstraintWeights_0=1) {
    alphaBoolBlend {blendInTime=0.3, blendOption=Linear, blendOutTime=0.3}
    alphaScaleBiasClamp {bInterpResult=true}
    boneToModify {boneName=dyn_shorts_in_l}
    constraintSetup [{transformType=Rotation, targetBone {boneName=calf_l}}]
    constraintWeights [1]
   }
   n40 = "Constraint - dyn_shorts_out_r (R)"(ComponentPose=@n39, Alpha=0.5, ConstraintWeights_0=1) {
    alphaBoolBlend {blendInTime=0.3, blendOption=Linear, blendOutTime=0.3}
    alphaScaleBiasClamp {bInterpResult=true}
    boneToModify {boneName=dyn_shorts_out_r}
    constraintSetup [{transformType=Rotation, targetBone {boneName=calf_r}}]
    constraintWeights [1]
   }
   n41 = "Constraint - dyn_shorts_out_l (R)"(ComponentPose=@n40, Alpha=0.5, ConstraintWeights_0=1) {
    alphaBoolBlend {blendInTime=0.3, blendOption=Linear, blendOutTime=0.3}
    alphaScaleBiasClamp {bInterpResult=true}
    boneToModify {boneName=dyn_shorts_out_l}
    constraintSetup [{transformType=Rotation, targetBone {boneName=calf_l}}]
    constraintWeights [1]
   }
   n42 = "Transform (Modify) Bone - Bone: dyn_shorts_bk_r"!(ComponentPose=@n41, bAlphaBoolEnabled:bool=True, Translation="-11.460,-4.090,-8.857", Rotation="-12.453,0.000,0.000", Scale="1,1,1") -> (Pose) {
    alphaInputType=Bool
    rotationMode=BMM_Additive
    rotationSpace=BCS_BoneSpace
    scaleMode=BMM_Additive
    scaleSpace=BCS_BoneSpace
    translationMode=BMM_Additive
    translationSpace=BCS_BoneSpace
    alphaBoolBlend {blendInTime=0.2, blendOption=Linear, blendOutTime=0.4}
    boneToModify {boneName=dyn_shorts_bk_r}
    rotation=<rot -12.453 0 0>
    translation=<-11.46 -4.09 -8.857>
   }
   n43 = "Transform (Modify) Bone - Bone: dyn_shorts_in_r"!(ComponentPose=@n42, bAlphaBoolEnabled:bool=True, Translation="-0.527,0.000,-3.788", Rotation="0.000,-0.000,17.848", Scale="1,1,1") -> (Pose) {
    alphaInputType=Bool
    rotationMode=BMM_Additive
    rotationSpace=BCS_BoneSpace
    scaleMode=BMM_Additive
    scaleSpace=BCS_BoneSpace
    translationMode=BMM_Additive
    translationSpace=BCS_BoneSpace
    alphaBoolBlend {blendInTime=0.2, blendOption=Linear, blendOutTime=0.4}
    boneToModify {boneName=dyn_shorts_in_r}
    rotation=<rot 0 0 17.848>
    translation {x=-0.527, z=-3.788}
   }
   n44 = "Transform (Modify) Bone - Bone: dyn_shorts_out_r"!(ComponentPose=@n43, bAlphaBoolEnabled:bool=True, Translation="-3.966,-0.117,-7.351", Rotation="-6.140,-8.403,-8.048", Scale="1,1,1") -> (Pose) {
    alphaInputType=Bool
    rotationMode=BMM_Additive
    rotationSpace=BCS_BoneSpace
    scaleMode=BMM_Additive
    scaleSpace=BCS_BoneSpace
    translationMode=BMM_Additive
    translationSpace=BCS_BoneSpace
    alphaBoolBlend {blendInTime=0.2, blendOption=Linear, blendOutTime=0.4}
    boneToModify {boneName=dyn_shorts_out_r}
    rotation=<rot -6.14 -8.403 -8.048>
    translation=<-3.966 -0.117 -7.351>
   }
   n45 = "Transform (Modify) Bone - Bone: dyn_shorts_fr_r"!(ComponentPose=@n44, bAlphaBoolEnabled:bool=True, Translation="0,0,0", Rotation="-2.170,11.018,10.178", Scale="1,1,1") -> (Pose) {
    alphaInputType=Bool
    rotationMode=BMM_Additive
    rotationSpace=BCS_BoneSpace
    scaleMode=BMM_Additive
    scaleSpace=BCS_BoneSpace
    translationMode=BMM_Additive
    translationSpace=BCS_BoneSpace
    alphaBoolBlend {blendInTime=0.2, blendOption=Linear, blendOutTime=0.4}
    boneToModify {boneName=dyn_shorts_fr_r}
    rotation=<rot -2.17 11.018 10.178>
   }
   n46 = "Transform (Modify) Bone - Bone: dyn_shorts_bk_l"!(ComponentPose=@n45, bAlphaBoolEnabled:bool=True, Translation="12.414,-0.000,-2.823", Rotation="-46.771,-0.000,0.000", Scale="1,1,1") -> (Pose) {
    alphaInputType=Bool
    rotationMode=BMM_Additive
    rotationSpace=BCS_BoneSpace
    scaleMode=BMM_Additive
    scaleSpace=BCS_BoneSpace
    translationMode=BMM_Additive
    translationSpace=BCS_BoneSpace
    alphaBoolBlend {blendInTime=0.2, blendOption=Linear, blendOutTime=0.4}
    boneToModify {boneName=dyn_shorts_bk_l}
    rotation=<rot -46.771 0 0>
    translation {x=12.414, z=-2.823}
   }
   n47 = "Transform (Modify) Bone - Bone: dyn_shorts_in_l"!(ComponentPose=@n46, bAlphaBoolEnabled:bool=True, Translation="0.000,0.363,1.892", Rotation="0.000,0.000,10.833", Scale="1,1,1") -> (Pose) {
    alphaInputType=Bool
    rotationMode=BMM_Additive
    rotationSpace=BCS_BoneSpace
    scaleMode=BMM_Additive
    scaleSpace=BCS_BoneSpace
    translationMode=BMM_Additive
    translationSpace=BCS_BoneSpace
    alphaBoolBlend {blendInTime=0.2, blendOption=Linear, blendOutTime=0.4}
    boneToModify {boneName=dyn_shorts_in_l}
    rotation=<rot 0 0 10.833>
    translation {y=0.363, z=1.892}
   }
   n48 = "Transform (Modify) Bone - Bone: dyn_shorts_out_l"!(ComponentPose=@n47, bAlphaBoolEnabled:bool=True, Translation="-0.000,-0.000,4.679", Rotation="0.000,-17.429,-0.000", Scale="1,1,1") -> (Pose) {
    alphaInputType=Bool
    rotationMode=BMM_Additive
    rotationSpace=BCS_BoneSpace
    scaleMode=BMM_Additive
    scaleSpace=BCS_BoneSpace
    translationMode=BMM_Additive
    translationSpace=BCS_BoneSpace
    alphaBoolBlend {blendInTime=0.2, blendOption=Linear, blendOutTime=0.4}
    boneToModify {boneName=dyn_shorts_out_l}
    rotation {yaw=-17.429}
    translation {z=4.679}
   }
   n49 = "Transform (Modify) Bone - Bone: dyn_shorts_bk_r"(ComponentPose=@n48, bAlphaBoolEnabled=True, Translation="-12.289,-3.278,21.945", Rotation="-3.066,0.296,0.822", Scale="1.000,1.481,1.000") {
    alphaInputType=Bool
    rotationMode=BMM_Additive
    rotationSpace=BCS_BoneSpace
    scaleMode=BMM_Additive
    scaleSpace=BCS_BoneSpace
    translationMode=BMM_Additive
    translationSpace=BCS_BoneSpace
    alphaBoolBlend {blendInTime=0.4, blendOption=Linear, blendOutTime=0.6}
    boneToModify {boneName=dyn_shorts_bk_r}
    rotation=<rot -3.066 0.296 0.822>
    scale {y=1.481}
    translation=<-12.289 -3.278 21.945>
   }
   n50 = "Transform (Modify) Bone - Bone: dyn_shorts_in_r"(ComponentPose=@n49, bAlphaBoolEnabled=True, Translation="0.133,-6.658,-3.704", Rotation="7.638,10.885,-66.089", Scale="1,1,1") {
    alphaInputType=Bool
    rotationMode=BMM_Additive
    rotationSpace=BCS_BoneSpace
    scaleMode=BMM_Additive
    scaleSpace=BCS_BoneSpace
    translationMode=BMM_Additive
    translationSpace=BCS_BoneSpace
    alphaBoolBlend {blendInTime=0.5, blendOption=Linear, blendOutTime=0.7}
    boneToModify {boneName=dyn_shorts_in_r}
    rotation=<rot 7.638 10.885 -66.089>
    translation=<0.133 -6.658 -3.704>
   }
   n51 = "Transform (Modify) Bone - Bone: dyn_shorts_out_r"(ComponentPose=@n50, bAlphaBoolEnabled=True, Translation="0.482,-3.259,1.022", Rotation="11.878,12.624,58.102", Scale="1,1,1") {
    alphaInputType=Bool
    rotationMode=BMM_Additive
    rotationSpace=BCS_BoneSpace
    scaleMode=BMM_Additive
    scaleSpace=BCS_BoneSpace
    translationMode=BMM_Additive
    translationSpace=BCS_BoneSpace
    alphaBoolBlend {blendInTime=0.5, blendOption=Linear, blendOutTime=0.7}
    boneToModify {boneName=dyn_shorts_out_r}
    rotation=<rot 11.878 12.624 58.102>
    translation=<0.482 -3.259 1.022>
   }
   n52 = "Transform (Modify) Bone - Bone: dyn_shorts_bk_l"(ComponentPose=@n51, bAlphaBoolEnabled=True, Translation="2.971,13.483,6.817", Rotation="17.131,29.712,57.309", Scale="1,1,1") {
    alphaInputType=Bool
    rotationMode=BMM_Additive
    rotationSpace=BCS_BoneSpace
    scaleMode=BMM_Additive
    scaleSpace=BCS_BoneSpace
    translationMode=BMM_Additive
    translationSpace=BCS_BoneSpace
    alphaBoolBlend {blendInTime=0.5, blendOption=Linear, blendOutTime=0.7}
    boneToModify {boneName=dyn_shorts_bk_l}
    rotation=<rot 17.131 29.712 57.309>
    translation=<2.971 13.483 6.817>
   }
   n53 = "Transform (Modify) Bone - Bone: dyn_shorts_in_l"(ComponentPose=@n52, bAlphaBoolEnabled=True, Translation="0.367,4.981,3.844", Rotation="-7.252,-6.154,-48.625", Scale="1,1,1") {
    alphaInputType=Bool
    rotationMode=BMM_Additive
    rotationSpace=BCS_BoneSpace
    scaleMode=BMM_Additive
    scaleSpace=BCS_BoneSpace
    translationMode=BMM_Additive
    translationSpace=BCS_BoneSpace
    alphaBoolBlend {blendInTime=0.5, blendOption=Linear, blendOutTime=0.7}
    boneToModify {boneName=dyn_shorts_in_l}
    rotation=<rot -7.252 -6.154 -48.625>
    translation=<0.367 4.981 3.844>
   }
   n54 = "Transform (Modify) Bone - Bone: dyn_shorts_out_l"(ComponentPose=@n53, bAlphaBoolEnabled=True, Translation="0.184,0.017,1.418", Rotation="1.074,-10.849,5.435", Scale="1,1,1") {
    alphaInputType=Bool
    rotationMode=BMM_Additive
    rotationSpace=BCS_BoneSpace
    scaleMode=BMM_Additive
    scaleSpace=BCS_BoneSpace
    translationMode=BMM_Additive
    translationSpace=BCS_BoneSpace
    alphaBoolBlend {blendInTime=0.5, blendOption=Linear, blendOutTime=0.7}
    boneToModify {boneName=dyn_shorts_out_l}
    rotation=<rot 1.074 -10.849 5.435>
    translation=<0.184 0.017 1.418>
   }
   n55 = "Transform (Modify) Bone - Bone: dyn_shorts_fr_r"(ComponentPose=@n54, bAlphaBoolEnabled=True, Translation="-0.650,-0.050,-0.020", Rotation="1.827,-9.326,8.629", Scale="1.000,1.472,1.000") {
    alphaInputType=Bool
    rotationMode=BMM_Additive
    rotationSpace=BCS_BoneSpace
    scaleMode=BMM_Additive
    scaleSpace=BCS_BoneSpace
    translationMode=BMM_Additive
    translationSpace=BCS_BoneSpace
    alphaBoolBlend {blendInTime=0.5, blendOption=Linear, blendOutTime=0.7}
    boneToModify {boneName=dyn_shorts_fr_r}
    rotation=<rot 1.827 -9.326 8.629>
    scale {y=1.472}
    translation=<-0.65 -0.05 -0.02>
   }
   n56 = "Transform (Modify) Bone - Bone: dyn_shorts_bk_r"(ComponentPose=@n55, bAlphaBoolEnabled=True, Translation="-4.337,0.156,17.125", Rotation="14.688,-0.002,-3.973", Scale="1.4,1,1") {
    alphaInputType=Bool
    rotationMode=BMM_Additive
    rotationSpace=BCS_BoneSpace
    scaleMode=BMM_Additive
    scaleSpace=BCS_BoneSpace
    translationMode=BMM_Additive
    translationSpace=BCS_BoneSpace
    alphaBoolBlend {blendInTime=0.5, blendOption=Linear, blendOutTime=0.7}
    boneToModify {boneName=dyn_shorts_bk_r}
    rotation=<rot 14.688 -0.002 -3.973>
    scale {x=1.4}
    translation=<-4.337 0.156 17.125>
   }
   n57 = "Transform (Modify) Bone - Bone: dyn_shorts_in_r"(ComponentPose=@n56, bAlphaBoolEnabled=True, Translation="0,0,0", Rotation="-15.713,-21.099,-12.734", Scale="1,1,1") {
    alphaInputType=Bool
    rotationMode=BMM_Additive
    rotationSpace=BCS_BoneSpace
    scaleMode=BMM_Additive
    scaleSpace=BCS_BoneSpace
    translationMode=BMM_Additive
    translationSpace=BCS_BoneSpace
    alphaBoolBlend {blendInTime=0.5, blendOption=Linear, blendOutTime=0.7}
    boneToModify {boneName=dyn_shorts_in_r}
    rotation=<rot -15.713 -21.099 -12.734>
   }
   n58 = "Transform (Modify) Bone - Bone: dyn_shorts_out_r"(ComponentPose=@n57, bAlphaBoolEnabled=True, Translation="0.168,-0.328,-1.243", Rotation="6.175,5.617,15.559", Scale="1,1,1") {
    alphaInputType=Bool
    rotationMode=BMM_Additive
    rotationSpace=BCS_BoneSpace
    scaleMode=BMM_Additive
    scaleSpace=BCS_BoneSpace
    translationMode=BMM_Additive
    translationSpace=BCS_BoneSpace
    alphaBoolBlend {blendInTime=0.5, blendOption=Linear, blendOutTime=0.7}
    boneToModify {boneName=dyn_shorts_out_r}
    rotation=<rot 6.175 5.617 15.559>
    translation=<0.168 -0.328 -1.243>
   }
   n59 = "Transform (Modify) Bone - Bone: dyn_shorts_bk_l"(ComponentPose=@n58, bAlphaBoolEnabled=True, Translation="-2,-1,8", Rotation="30,-20,0", Scale="1,1,1") {
    alphaInputType=Bool
    rotationMode=BMM_Additive
    rotationSpace=BCS_BoneSpace
    scaleMode=BMM_Additive
    scaleSpace=BCS_BoneSpace
    translationMode=BMM_Additive
    translationSpace=BCS_BoneSpace
    alphaBoolBlend {blendInTime=0.5, blendOption=Linear, blendOutTime=0.7}
    boneToModify {boneName=dyn_shorts_bk_l}
    rotation {pitch=30, yaw=-20}
    translation=<-2 -1 8>
   }
   n60 = "Transform (Modify) Bone - Bone: dyn_shorts_in_l"(ComponentPose=@n59, bAlphaBoolEnabled=True, Translation="0,0,0", Rotation="-0.134,16.622,-14.355", Scale="1,1,1") {
    alphaInputType=Bool
    rotationMode=BMM_Additive
    rotationSpace=BCS_BoneSpace
    scaleMode=BMM_Additive
    scaleSpace=BCS_BoneSpace
    translationMode=BMM_Additive
    translationSpace=BCS_BoneSpace
    alphaBoolBlend {blendInTime=0.5, blendOption=Linear, blendOutTime=0.7}
    boneToModify {boneName=dyn_shorts_in_l}
    rotation=<rot -0.134 16.622 -14.355>
   }
   n61 = "Transform (Modify) Bone - Bone: dyn_shorts_out_l"(ComponentPose=@n60, bAlphaBoolEnabled=True, Translation="0,0,0", Rotation="-0.152,-0.101,20", Scale="1,1,1") {
    alphaInputType=Bool
    rotationMode=BMM_Additive
    rotationSpace=BCS_BoneSpace
    scaleMode=BMM_Additive
    scaleSpace=BCS_BoneSpace
    translationMode=BMM_Additive
    translationSpace=BCS_BoneSpace
    alphaBoolBlend {blendInTime=0.5, blendOption=Linear, blendOutTime=0.7}
    boneToModify {boneName=dyn_shorts_out_l}
    rotation=<rot -0.152 -0.101 20>
   }
   n62 = "Transform (Modify) Bone - Bone: dyn_shorts_fr_r"(ComponentPose=@n61, bAlphaBoolEnabled=True, Translation="0,0,0", Rotation="0,0,0", Scale="1,1,1") {
    alphaInputType=Bool
    rotationMode=BMM_Additive
    rotationSpace=BCS_BoneSpace
    scaleMode=BMM_Additive
    scaleSpace=BCS_BoneSpace
    translationMode=BMM_Additive
    translationSpace=BCS_BoneSpace
    alphaBoolBlend {blendInTime=0.5, blendOption=Linear, blendOutTime=0.7}
    boneToModify {boneName=dyn_shorts_fr_r}
   }
   n63 = "Constraint - dyn_bag_01_MSTR (PPPPPPPP)"!(
    ComponentPose=@n62
    Alpha=1
    ConstraintWeights_0=0.345
    ConstraintWeights_1:float=0.0236
    ConstraintWeights_2:float=0.0364
    ConstraintWeights_3:float=0.0284
    ConstraintWeights_4:float=0.4247
    ConstraintWeights_5:float=0.0311
    ConstraintWeights_6:float=0.1107
    ConstraintWeights_7:float=0.0002
   ) -> (Pose) {
    alphaBoolBlend {blendOption=Linear}
    boneToModify {boneName=dyn_bag_01_MSTR}
    constraintSetup [
     {targetBone {boneName=pelvis}}
     {targetBone {boneName=spine_01}}
     {targetBone {boneName=spine_02}}
     {targetBone {boneName=spine_03}}
     {targetBone {boneName=thigh_twist_01_l}}
     {targetBone {boneName=deform_glute_l}}
     {targetBone {boneName=deform_groin_l}}
     {targetBone {boneName=thigh_twist_01_r}}
    ]
    constraintWeights [1,1,1,1,1,1,1,1]
   }
   n64 = "Constraint - dyn_bag_02_MSTR (PPPPPPP)"!(
    ComponentPose=@n63
    Alpha=1
    ConstraintWeights_0=0.4384
    ConstraintWeights_1:float=0.0274
    ConstraintWeights_2:float=0.0436
    ConstraintWeights_3:float=0.0353
    ConstraintWeights_4:float=0.2865
    ConstraintWeights_5:float=0.0757
    ConstraintWeights_6:float=0.0932
   ) -> (Pose) {
    alphaBoolBlend {blendOption=Linear}
    boneToModify {boneName=dyn_bag_02_MSTR}
    constraintSetup [
     {targetBone {boneName=pelvis}}
     {targetBone {boneName=spine_01}}
     {targetBone {boneName=spine_02}}
     {targetBone {boneName=spine_03}}
     {targetBone {boneName=thigh_twist_01_l}}
     {targetBone {boneName=deform_glute_l}}
     {targetBone {boneName=deform_groin_l}}
    ]
    constraintWeights [1,1,1,1,1,1,1]
   }
   n65 = "Constraint - dyn_bag_01_MSTR (R)"(ComponentPose=@n64, Alpha=1, ConstraintWeights_0=1) {
    alphaBoolBlend {blendOption=Linear}
    boneToModify {boneName=dyn_bag_01_MSTR}
    constraintSetup [{transformType=Rotation, targetBone {boneName=thigh_l}}]
    constraintWeights [1]
   }
   n66 = "Constraint - dyn_bag_02_MSTR (R)"(ComponentPose=@n65, Alpha=1, ConstraintWeights_0=1) {
    alphaBoolBlend {blendOption=Linear}
    boneToModify {boneName=dyn_bag_02_MSTR}
    constraintSetup [{transformType=Rotation, targetBone {boneName=thigh_l}}]
    constraintWeights [1]
   }
   n67 = "Constraint - dyn_bag_02_MSTR (PPPPPPP)"!(
    ComponentPose=@n66
    bAlphaBoolEnabled:bool=True
    ConstraintWeights_0=0.44
    ConstraintWeights_1=0.0219
    ConstraintWeights_2=0.0359
    ConstraintWeights_3=0.0305
    ConstraintWeights_4=0.2817
    ConstraintWeights_5=0.1217
    ConstraintWeights_6=0.0683
   ) -> (Pose) {
    alphaInputType=Bool
    alphaBoolBlend {blendOption=Linear}
    boneToModify {boneName=dyn_bag_02_MSTR}
    constraintSetup [
     {targetBone {boneName=pelvis}}
     {targetBone {boneName=spine_01}}
     {targetBone {boneName=spine_02}}
     {targetBone {boneName=spine_03}}
     {targetBone {boneName=thigh_twist_01_l}}
     {targetBone {boneName=deform_glute_l}}
     {targetBone {boneName=deform_groin_l}}
    ]
    constraintWeights [1,1,1,1,1,1,1]
   }
   n68 = "Transform (Modify) Bone - Bone: dyn_bag_01_MSTR"!(ComponentPose=@n67, bAlphaBoolEnabled:bool=True, Translation="0.830,1.631,-1.044", Rotation="9.859,-3.565,60.310", Scale="1,1,1") -> (Pose) {
    alphaInputType=Bool
    rotationMode=BMM_Additive
    rotationSpace=BCS_BoneSpace
    scaleMode=BMM_Additive
    scaleSpace=BCS_BoneSpace
    translationMode=BMM_Additive
    translationSpace=BCS_BoneSpace
    alphaBoolBlend {blendInTime=0.4, blendOption=Linear, blendOutTime=0.4}
    boneToModify {boneName=dyn_bag_01_MSTR}
    rotation=<rot 9.859 -3.565 60.31>
    translation=<0.83 1.631 -1.044>
   }
   n69 = "Transform (Modify) Bone - Bone: dyn_bag_02_MSTR"!(ComponentPose=@n68, bAlphaBoolEnabled:bool=True, Translation="-0.000,0.000,-2.049", Rotation="-2.718,0.822,-0.040", Scale="1,1,1") -> (Pose) {
    alphaInputType=Bool
    rotationMode=BMM_Additive
    rotationSpace=BCS_BoneSpace
    scaleMode=BMM_Additive
    scaleSpace=BCS_BoneSpace
    translationMode=BMM_Additive
    translationSpace=BCS_BoneSpace
    alphaBoolBlend {blendInTime=0.4, blendOption=Linear, blendOutTime=0.4}
    boneToModify {boneName=dyn_bag_02_MSTR}
    rotation=<rot -2.718 0.822 -0.04>
    translation {z=-2.049}
   }
   n70 = "Constraint - dyn_Map_MSTR (PPPPPPPP)"!(
    ComponentPose=@n69
    Alpha=1
    ConstraintWeights_0=0.4335
    ConstraintWeights_1:float=0.0235
    ConstraintWeights_2:float=0.037
    ConstraintWeights_3:float=0.0298
    ConstraintWeights_4:float=0.0002
    ConstraintWeights_5:float=0.2966
    ConstraintWeights_6:float=0.0719
    ConstraintWeights_7:float=0.1076
   ) -> (Pose) {
    alphaBoolBlend {blendOption=Linear}
    boneToModify {boneName=dyn_Map_MSTR}
    constraintSetup [
     {targetBone {boneName=pelvis}}
     {targetBone {boneName=spine_01}}
     {targetBone {boneName=spine_02}}
     {targetBone {boneName=spine_03}}
     {targetBone {boneName=thigh_twist_01_l}}
     {targetBone {boneName=thigh_twist_01_r}}
     {targetBone {boneName=deform_glute_r}}
     {targetBone {boneName=deform_groin_r}}
    ]
    constraintWeights [1,1,1,1,1,1,1,1]
   }
   n71 = "Constraint - dyn_Map_MSTR (R)"(ComponentPose=@n70, Alpha=0.7, ConstraintWeights_0=1) {
    alphaBoolBlend {blendOption=Linear}
    boneToModify {boneName=dyn_Map_MSTR}
    constraintSetup [{transformType=Rotation, targetBone {boneName=thigh_r}}]
    constraintWeights [1]
   }
   n72 = Control Rig!(
    Source=Component To Local(ComponentPose=@n71)
    Alpha
    Pistol:double=Get isAimingPistol() -> (isAimingPistol:double)
    Rifle:double=Get isAimingRifle() -> (isAimingRifle:double)
    Skydiving:double=Get SkydivingFloat() -> (SkydivingFloat:double)
    OnGround:double=Get OnGroundFloat() -> (OnGroundFloat:double)
    Running:double=Get Running() -> (Running:double)
   ) -> (Pose) {
    alphaBoolBlend {blendOption=Linear}
    controlRigAssetReference {blueprintRigClass=&/M_MED_BananaAdventure_Jacket_CtrlRig.M_MED_BananaAdventure_Jacket_CtrlRig_C}
   }
   n73 = "Constraint - dyn_Zipper_MSTR (PPPPPPPP)"(
    ComponentPose=Local To Component(LocalPose=@n72)
    Alpha=1
    ConstraintWeights_0=0.0013
    ConstraintWeights_1=0.0578
    ConstraintWeights_2=0.7724
    ConstraintWeights_3=0.0072
    ConstraintWeights_4=0.0031
    ConstraintWeights_5=0.1045
    ConstraintWeights_6=0.052
    ConstraintWeights_7=0.0017
   ) {
    alphaBoolBlend {blendOption=Linear}
    boneToModify {boneName=dyn_Zipper_MSTR}
    constraintSetup [
     {targetBone {boneName=deform_pec_l}}
     {targetBone {boneName=dyn_jacket_a_01}}
     {targetBone {boneName=dyn_jacket_a_02}}
     {targetBone {boneName=dyn_jacket_c_01}}
     {targetBone {boneName=dyn_jacket_c_02}}
     {targetBone {boneName=dyn_jacket_b_01}}
     {targetBone {boneName=dyn_jacket_b_02}}
     {targetBone {boneName=dfrm_clavicle_l}}
    ]
    constraintWeights [1,1,1,1,1,1,1,1]
   }
   n74 = Look At - Bone: dyn_Zipper_01(ComponentPose=@n73, Alpha=1) {alphaBoolBlend {blendOption=Linear}, boneToModify {boneName=dyn_Zipper_01}, lookAtTarget {boneReference {boneName=pelvis}}}
   n75 = Look At - Bone: dyn_Zipper_01!(ComponentPose=@n74, bAlphaBoolEnabled:bool=True) -> (Pose) {
    alphaInputType=Bool
    alphaBoolBlend {blendOption=Linear}
    boneToModify {boneName=dyn_Zipper_01}
    lookAtLocation=<49.7229 241.09 135.9188>
    lookAtTarget {boneReference {boneName=pelvis}}
   }
   Output Pose(Result=Component To Local(ComponentPose=@n75)) {layerGroup=None}
   n76 = Control Rig!(Source=Copy Pose From Mesh() {bUseAttachedParent=true}, Alpha) -> (Pose) {
    alphaBoolBlend {blendOption=Linear}
    controlRigAssetReference {blueprintRigClass=&/M_MED_BananaAdventure_CtrlRig.M_MED_BananaAdventure_CtrlRig_C}
   }
   n77 = RigidBody(ComponentPose=Local To Component(LocalPose=@n76), OverrideWorldGravity="0,0,0") {
    bClampLinearTranslationLimitToRefPose=true
    bOverrideWorldGravity=true
    overridePhysicsAsset=&/M_MED_BananaAdventure_RightShorts_Physics
    simulationSpace=BaseBoneSpace
    alphaBoolBlend {blendOption=Linear}
    baseBoneRef {boneName=thigh_twist_01_r}
    componentAppliedLinearAccClamp=<1000 1000 1000>
    componentLinearAccScale=<0.2 0.2 0.2>
    componentLinearVelScale=<0.2 0.2 0.2>
    overrideWorldGravity {z=-980}
    simSpaceSettings {
     maxAngularAcceleration=1000
     maxAngularVelocity=3
     maxLinearAcceleration=700
     maxLinearVelocity=500
     velocityScaleZ=0
     worldAlpha=1
     externalLinearDragV=<0.2 0.2>
    }
   }
   n78 = RigidBody(ComponentPose=@n77, OverrideWorldGravity="0,0,0") {
    bClampLinearTranslationLimitToRefPose=true
    bOverrideWorldGravity=true
    overridePhysicsAsset=&/M_MED_BananaAdventure_LeftShorts_Physics
    simulationSpace=BaseBoneSpace
    alphaBoolBlend {blendOption=Linear}
    baseBoneRef {boneName=thigh_twist_01_l}
    componentAppliedLinearAccClamp=<1000 1000 1000>
    componentLinearAccScale=<0.2 0.2 0.2>
    componentLinearVelScale=<0.2 0.2 0.2>
    overrideWorldGravity {z=-980}
    simSpaceSettings {
     maxAngularAcceleration=1000
     maxAngularVelocity=3
     maxLinearAcceleration=700
     maxLinearVelocity=500
     velocityScaleZ=0
     worldAlpha=1
     externalLinearDragV=<0.2 0.2>
    }
   }
   Save cached pose 'PreJacketSim'(Pose=Component To Local(ComponentPose=@n78)) {cacheName=PreJacketSim}
  ]}
 EventGraph {nodes [
   Event BlueprintUpdateAnimation()
   > n1 = Sequence!(execute) -> (then_0, then_1, then_2:Exec, then_3:Exec, then_4:Exec, then_5:Exec, then_6:Exec, then_7:Exec, then_8:Exec, then_9:Exec, then_10:Exec)
   > n2 = Gravity Override State Function DEPRECATED(CharacterPartAnimInstance=Self-Reference(), GameTimeInSeconds=GetGameTimeInSeconds(), GravityOverrideParams=Get Gravity Override Params())
   > Set GravityOverride(GravityOverride=@n2)
   n3 = GetOwnerSkeletalMeshComponent()
   n4 = BreakRotator(InRot=BreakTransform(InTransform=GetDeltaTransformFromRefPose(self=@n3, BoneName=head, BaseName=neck_02)).Rotation)
   n5 = Get Y percentage() -> (Y percentage:float)
   n6 = Get X percentage() -> (X percentage:float)
   n7 = Get Z percentage() -> (Z percentage:float)
   Set head rot(
    execute=@n1.then_1
    head rot:Rotator=MakeRotator(
     Roll=float * float!(A=@n4, B=-1, C:double=@n6) -> (ReturnValue)
     Pitch=float * float(A=@n4.Pitch, B=-1, C=@n5)
     Yaw=float * float(A=@n4.Yaw, B=-1, C=@n7)
    )
   ) -> (Output_Get:Rotator)
   n8 = Get bIsSkyDiving()
   n9 = NOT Boolean(A=Get bIsCrouching())
   n10 = Get bIsFalling()
   n11 = Get bIsJumping()
   n12 = Branch(execute=@n1.then_10, Condition=AND Boolean(A=OR Boolean(A=@n11, B=@n10, C=@n9), B=NOT Boolean(A=@n8)))
   > Set CrouchFloat_01(CrouchFloat_01:double=1) -> (Output_Get:double)
   > Set CrouchFloat_02(CrouchFloat_02:double=0.5) -> (Output_Get:double)
   Set CrouchFloat_01(execute=@n12.else, CrouchFloat_01=0)
   > Set CrouchFloat_02(CrouchFloat_02=0)
   n13 = Cast To FortPlayerAnimInstance(execute=@n1.then_2, Object=GetAnimInstance(self=GetOwnerSkeletalMeshComponent()))
   > Set isAimingPistol(isAimingPistol:double=0) -> (Output_Get:double)
   > Set isPickaxe(isPickaxe:double=0) -> (Output_Get:double)
   > Set isAimingRifle(isAimingRifle:double=0) -> (Output_Get:double)
   > Set isPickaxeBoo(isPickaxeBoo:bool=false) -> (Output_Get:bool)
   > n14 = Switch on EFortWeaponCoreAnimation!(execute, Selection:EFortWeaponCoreAnimation Enum=Get WeaponCoreAnim(self=@n13)) -> (
    Melee:Exec
    Pistol:Exec
    Shotgun:Exec
    PaperBlueprint:Exec
    Rifle:Exec
    MeleeOneHand:Exec
    MachinePistol:Exec
    RocketLauncher:Exec
    GrenadeLauncher:Exec
    GoingCommando:Exec
    AssaultRifle:Exec
    TacticalShotgun:Exec
    SniperRifle:Exec
    TrapPlacement:Exec
    ShoulderLauncher:Exec
    AbilityDecoTool:Exec
    Crossbow:Exec
    C4:Exec
    RemoteControl:Exec
    DualWield:Exec
    AR_BullPup:Exec
    AR_ForwardGrip:Exec
    MedPackPaddles:Exec
    SMG_P90:Exec
    AR_DrumGun:Exec
    Consumable_Small:Exec
    Consumable_Large:Exec
    Balloon:Exec
    MountedTurret:Exec
    CreativeTool:Exec
    ExplosiveBow:Exec
    AshtonIndigo:Exec
    AshtonChicago:Exec
    MeleeDualWield:Exec
    Unarmed:Exec
   )
   n15 = Get bIsOnGround() -> (bIsOnGround:bool)
   Branch(
    execute=[
     @n14.AssaultRifle
     @n14.Crossbow
     @n14.GrenadeLauncher
     @n14.Rifle
     @n14.RocketLauncher
     @n14.Shotgun
     @n14.SniperRifle
     @n14.TacticalShotgun
    ]
    Condition=@n15
   )
   > Set isAimingRifle(isAimingRifle=1)
   Branch(execute=[@n14.MachinePistol, @n14.Pistol], Condition=@n15)
   > Set isAimingPistol(isAimingPistol=1)
   Branch(execute=@n14.Melee, Condition=@n15)
   > Set isPickaxe(isPickaxe=1)
   > Set isPickaxeBoo(isPickaxeBoo=true)
   n16 = Branch(execute=@n1.then_7, Condition=Get bIsOnGround())
   > Set OnGroundFloat(OnGroundFloat:double=1) -> (Output_Get:double)
   Set OnGroundFloat(execute=@n16.else, OnGroundFloat=0)
   n17 = Branch(execute=@n1.then_8, Condition=Get bIsSkydiveFloating())
   > Set SkydivingFloatFloat(SkydivingFloatFloat:double=1) -> (Output_Get:double)
   Set SkydivingFloatFloat(execute=@n17.else, SkydivingFloatFloat=0)
   n18 = Get bIsFalling()
   n19 = Get bIsJumping()
   n20 = Get bIsCrouching()
   n21 = Get bIsAccelerating2D()
   n22 = Branch(
    execute=@n1.then_9
    Condition=AND Boolean!(A=NOT Boolean(A=@n20), B=@n21, C:bool=NOT Boolean(A=@n19), D:bool=NOT Boolean(A=@n18)) -> (ReturnValue)
   )
   > Set Running(Running:double=1) -> (Output_Get:double)
   Set Running(execute=@n22.else, Running=0)
   n23 = Get bIsAccelerating2D()
   n24 = Branch(execute=@n1.then_5, Condition=AND Boolean!(A=Get bIsCrouching(), B=NOT Boolean(A=@n23)) -> (ReturnValue))
   > Set NewCrouching(NewCrouching:bool=true) -> (Output_Get:bool)
   > Set NewCrouchingFloat(NewCrouchingFloat:double=1) -> (Output_Get:double)
   Set NewCrouching(execute=@n24.else, NewCrouching=false)
   > Set NewCrouchingFloat(NewCrouchingFloat=0)
   n25 = Branch(execute=@n1.then_4, Condition=AND Boolean(A=Get bIsCrouching(), B=Get bIsAccelerating2D()))
   > Set CrouchingAndMoving(CrouchingAndMoving:bool=true) -> (Output_Get:bool)
   > Set CrouchingAndMovingFloat(CrouchingAndMovingFloat:double=1) -> (Output_Get:double)
   Set CrouchingAndMoving(execute=@n25.else, CrouchingAndMoving=false)
   > Set CrouchingAndMovingFloat(CrouchingAndMovingFloat=0)
   n26 = Get bIsAccelerating2D()
   n27 = Get bIsJumping()
   n28 = Branch(
    execute=@n1.then_3
    Condition=OR Boolean!(A=OR Boolean(A=@n27, B=false), B=AND Boolean!(A=false, B=false, C:bool=@n26) -> (ReturnValue), C) -> (ReturnValue)
   )
   > Set NewJumping(NewJumping:bool=true) -> (Output_Get:bool)
   Set NewJumping(execute=@n28.else, NewJumping=false)
   n29 = Branch(execute=@n1.then_6, Condition=OR Boolean!(A=Get bIsSkyDiving(), B=Get bIsParachuteOpened()) -> (ReturnValue))
   > Set SkydivingFloat(SkydivingFloat:double=1) -> (Output_Get:double)
   Set SkydivingFloat(execute=@n29.else, SkydivingFloat=0)
   OR Boolean!(A=NOT Boolean(A=Get bIsSkyDiving()), B=false, C:bool) -> (ReturnValue)
   AND Boolean(A=false, B=false)
   "Byte > Byte"(A:Byte, B:bool=@n21) -> (ReturnValue:bool=~) {is=title K2Node_PromotableOperator}
   Get bIsFalling()
   Get bIsFalling()
  ]}
]
variables [
 CrouchFloat_01 {type=number}
 CrouchFloat_02 {type=number}
 CrouchingAndMoving {type=boolean}
 CrouchingAndMovingFloat {type=number}
 Gravity Override Params {
  type=GravityOverrideParamsStruct
  default {skydive_FloatingDown {gravityOverride=<0 -700 350>}, skydive_GoingUp {gravityOverride=<0 -300 100>}}
 }
 GravityOverride {type=Vector}
 head rot {type=Rotator}
 isAimingPistol {type=number}
 isAimingRifle {type=number}
 isPickaxe {type=number}
 isPickaxeBoo {type=boolean}
 NewCrouching {type=boolean}
 NewCrouchingFloat {type=number}
 NewJumping {type=boolean}
 OnGroundFloat {type=number}
 Running {type=number}
 SkydivingFloat {type=number}
 SkydivingFloatFloat {type=number}
 X percentage {type=number}
 Y percentage {type=number}
 Z percentage {type=number}
]
