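# Physics asset for the face accessory of the medium male BananaAdventure player body, simulating its dangling left and right flaps.
# Capsule bodies cover the flap bones, with a large capsule on neck_01 as the anchor.
# The left flap is a three-body chain with narrow angular limits between segments and a free-rotating base joint at the neck.
# The right flap is a single body constrained directly to the neck with narrow limits.
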PhysicsAsset /BRCosmetics/Characters/Player/Male/Medium/Bodies/M_MED_BananaAdventure/Meshes/Parts/M_MED_BananaAdventure_FaceAcc_Physics
bodies {
 dyn_flap_01_l [{shapeType=Capsule, length=5.75, radius=3, center=<0.3728 5.4183 1.8167>, rotation=<rot 0 24.957 70.1923>}]
 dyn_flap_02_l [{shapeType=Capsule, length=5.75, radius=3, center=<5.5001 0 0>, rotation=<rot 90 0 0>}]
 dyn_flap_Base_l [{shapeType=Capsule, length=0.505, radius=0.505, shapeName=dyn_flap_Base_l_capsule, rotation=<rot 0 -90 0>}]
 dyn_flap_r [{shapeType=Capsule, length=8.25, radius=3.5, center=<5 0 0>, rotation=<rot 90 0 0>}]
 neck_01 [{shapeType=Capsule, length=40.75, radius=14.5, rotation=<rot 90 10.6197 23.0097>}]
}
constraints [
 {bone1Name=dyn_flap_01_l, bone2Name=dyn_flap_Base_l, swing1LimitDegrees=15, swing2LimitDegrees=5, twistLimitDegrees=5}
 {bone1Name=dyn_flap_02_l, bone2Name=dyn_flap_01_l, swing1LimitDegrees=15, swing2LimitDegrees=5, twistLimitDegrees=5}
 {
  bone1Name=dyn_flap_Base_l
  bone2Name=neck_01
  swing1LimitDegrees=45
  swing1Motion=Free
  swing2LimitDegrees=45
  swing2Motion=Free
  twistLimitDegrees=45
  twistMotion=Free
 }
 {bone1Name=dyn_flap_r, bone2Name=neck_01, swing1LimitDegrees=10, swing2LimitDegrees=5, twistLimitDegrees=5}
]
