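# Animation Blueprint for the face accessory part of the BananaAdventure character skin, driving its facial pose, secondary physics and mouth material.
# The pose and curves copied from the attached parent mesh are layered with a looping banana facial idle, weighted by one minus DisableFaceOverride.
# A RigidBody node then simulates the result in root-bone space with its own physics asset and gravity.
# Each update, the mouth animation frame parameter on the dynamic material switches according to whether an emote is playing.
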AnimBlueprint /BRCosmetics/Characters/Player/Male/Medium/Bodies/M_MED_BananaAdventure/Meshes/Parts/M_MED_BananaAdventure_FaceAcc_AnimBP
target_skeleton=&/M_MED_BananaAdventure_FaceAcc_Skeleton
graphs [
 AnimGraph {nodes [
   n1 = Copy Pose From Mesh() {bCopyCurves=true, bUseAttachedParent=true}
   n2 = Sequence Player 'Banana_Male_Facial_Idle'() {sequence=/Game/Animation/Game/MainPlayer/Facial/Banana/Banana_Male_Facial_Idle}
   n3 = float - float(A=1, B=Get DisableFaceOverride() -> (DisableFaceOverride:float))
   n4 = RigidBody(ComponentPose=Local To Component(LocalPose=Layered blend per bone(BasePose=@n1, BlendPoses_0=@n2, BlendWeights_0=@n3))) {
    bClampLinearTranslationLimitToRefPose=true
    bOverrideWorldGravity=true
    overridePhysicsAsset=&/M_MED_BananaAdventure_FaceAcc_Physics
    simulationSpace=BaseBoneSpace
    alphaBoolBlend {blendOption=Linear}
    baseBoneRef {boneName=root}
    componentAppliedLinearAccClamp=<800 800 800>
    componentLinearAccScale=<0.4 0.4 0.5>
    overrideWorldGravity {z=-980}
   }
   Output Pose(Result=Component To Local(ComponentPose=@n4)) {layerGroup=None}
  ]}
 EventGraph {nodes [
   Event BlueprintUpdateAnimation()
   > n1 = Get Dynamic Material Instance(SkeletalMeshComponent=GetPartSkeletalMeshComponent())
   > n2 = Branch(Condition=Get bIsPlayingEmote())
   > SetScalarParameterValue(self=@n1, ParameterName=Smile Mouth animation Frame, Value=8)
   SetScalarParameterValue(execute=@n2.else, self=@n1, ParameterName=Smile Mouth animation Frame, Value=3)
  ]}
]
variables []
